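# Masked, unlit, two-sided inverted-hull anime outline material for the DuckCoast character's head.
# The shared outline function feeds cloth wind, whose opacity mask also runs through an eyeball DF-line masking function borrowed from SilentNovel Vane.
# A UV1 step multiplied by Fresnel blends that masked result back toward the original mask.
# Character PostFX takes the result with pre-skinned position, distance blend and Fresnel; a static debug switch offers an unmasked variant.
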Material /BRCosmetics/Characters/Player/Male/Medium/Bodies/M_MED_DuckCoast/Materials/M_DuckCoast_Head_InvHullOutline
graph {
 expressions {
  MaterialExpressionComponentMask_0 {class=MaterialExpressionComponentMask, inputs [None {from=MaterialExpressionTextureCoordinate_1}], properties {r=true}}
  MaterialExpressionComposite_0 {class=MaterialExpressionComposite, inputs [], properties {subgraphName=Construct Base Inputs}}
  MaterialExpressionConstant_1 {class=MaterialExpressionConstant}
  MaterialExpressionConstant_2 {class=MaterialExpressionConstant, properties {r=1}}
  MaterialExpressionFresnel_0 {
   class=MaterialExpressionFresnel
   inputs [
    ExponentIn {from="", output=""}
    BaseReflectFractionIn {from="", output=""}
    Normal {from="", output=""}
   ]
   properties {baseReflectFraction=0, exponent=1}
  }
  MaterialExpressionFresnel_1 {
   class=MaterialExpressionFresnel
   inputs [
    ExponentIn {from="", output=""}
    BaseReflectFractionIn {from="", output=""}
    Normal {from="", output=""}
   ]
   properties {baseReflectFraction=0, exponent=2}
  }
  MaterialExpressionGetMaterialAttributes_0 {
   class=MaterialExpressionGetMaterialAttributes
   inputs [None {from=MaterialExpressionMaterialFunctionCall_1, output=Result}]
   properties {attributeGetTypes [#1, #2]}
  }
  MaterialExpressionLinearInterpolate_1 {
   class=MaterialExpressionLinearInterpolate
   inputs [
    A {from=MaterialExpressionMaterialFunctionCall_2, output=Result}
    B {from=MaterialExpressionGetMaterialAttributes_0, output=Opacity Mask}
    Alpha {from=MaterialExpressionMultiply_0}
   ]
  }
  MaterialExpressionMaterialFunctionCall_0 {
   class=MaterialExpressionMaterialFunctionCall
   inputs []
   properties {materialFunction=/Game/Global/Materials/Character_Parent_2023/Anime_Parent_2023/MaterialFunctions/MFC_AnimeOutline_InvHull_2023}
  }
  MaterialExpressionMaterialFunctionCall_1 {
   class=MaterialExpressionMaterialFunctionCall
   inputs [
    Material {from=MaterialExpressionMaterialFunctionCall_0, output=Result}
    FX {from=MaterialExpressionMaterialFunctionCall_0, output=Result}
   ]
   properties {materialFunction=/Game/Packages/Fortress_SharedMaterials/Anime_Master_Material/MFC_ClothWind}
  }
  MaterialExpressionMaterialFunctionCall_14 {
   class=MaterialExpressionMaterialFunctionCall
   inputs [Material {from=""}]
   properties {materialFunction=/Game/Packages/Fortress_SharedMaterials/Base_Material_Functions/CharacterBrightnessandScale_Balance}
  }
  MaterialExpressionMaterialFunctionCall_16 {
   class=MaterialExpressionMaterialFunctionCall
   inputs [
    Material {from=MaterialExpressionSetMaterialAttributes_0}
    M {from=MaterialExpressionReroute_8, output=M}
    Pre Skinned Local Position {from=MaterialExpressionReroute_9, output=Pre Skinned Local Position}
    Distance Blend {from=MaterialExpressionReroute_10, output=Distance Blend}
    Fresnel {from=MaterialExpressionReroute_11, output=Fresnel}
   ]
   properties {materialFunction=&MF_CharacterParent_2023_PostFX}
  }
  MaterialExpressionMaterialFunctionCall_2 {
   class=MaterialExpressionMaterialFunctionCall
   inputs [DFLines {from=MaterialExpressionGetMaterialAttributes_0, output=Opacity Mask}]
   properties {materialFunction=/BRCosmetics/Characters/Player/Male/Medium/Bodies/M_MED_SilentNovel_Vane/Materials/MF_EyeballsDFLineMasking}
  }
  MaterialExpressionMultiply_0 {
   class=MaterialExpressionMultiply
   inputs [
    A {from=MaterialExpressionSaturate_0}
    B {from=MaterialExpressionFresnel_1}
   ]
  }
  MaterialExpressionPinBase_0 {class=MaterialExpressionPinBase, inputs [], properties {}}
  MaterialExpressionPinBase_1 {
   class=MaterialExpressionPinBase
   inputs [
    M {from=MaterialExpressionConstant_1, output=""}
    Pre Skinned Local Position {from=MaterialExpressionShaderStageSwitch_0, output=""}
    Distance Blend {from=MaterialExpressionMaterialFunctionCall_14, output=Distance Blend}
    Fresnel {from=MaterialExpressionFresnel_0, output=""}
   ]
   properties {reroutePins [
     M {expression=&self:MaterialGraph_1.MaterialGraphNode_Composite_0.Construct Base Inputs.MaterialExpressionReroute_8}
     Pre Skinned Local Position {expression=&self:MaterialGraph_1.MaterialGraphNode_Composite_0.Construct Base Inputs.MaterialExpressionReroute_9}
     Distance Blend {expression=&self:MaterialGraph_1.MaterialGraphNode_Composite_0.Construct Base Inputs.MaterialExpressionReroute_10}
     Fresnel {expression=&self:MaterialGraph_1.MaterialGraphNode_Composite_0.Construct Base Inputs.MaterialExpressionReroute_11}
    ]}
  }
  MaterialExpressionPreSkinnedPosition_0 {class=MaterialExpressionPreSkinnedPosition, inputs [], properties {}}
  MaterialExpressionSaturate_0 {class=MaterialExpressionSaturate, inputs [None {from=MaterialExpressionSign_0}]}
  MaterialExpressionSetMaterialAttributes_0 {
   class=MaterialExpressionSetMaterialAttributes
   inputs [
    MaterialAttributes {from=MaterialExpressionGetMaterialAttributes_0, output=MaterialAttributes}
    Opacity Mask {from=MaterialExpressionLinearInterpolate_1, output=""}
   ]
   properties {attributeSetTypes [#1]}
  }
  MaterialExpressionSetMaterialAttributes_3 {
   class=MaterialExpressionSetMaterialAttributes
   inputs [
    MaterialAttributes {from="", output=""}
    Opacity Mask {from=MaterialExpressionConstant_2, output=""}
    Emissive Color {from="", output=""}
    World Position Offset {from=MaterialExpressionGetMaterialAttributes_0, output=World Position Offset}
   ]
   properties {attributeSetTypes [#1, #3, #2]}
  }
  MaterialExpressionShaderStageSwitch_0 {
   class=MaterialExpressionShaderStageSwitch
   inputs [
    PixelShader {from=MaterialExpressionVertexInterpolator_0, output=PS}
    VertexShader {from=MaterialExpressionPreSkinnedPosition_0, output=""}
   ]
   properties {}
  }
  MaterialExpressionSign_0 {class=MaterialExpressionSign, inputs [None {from=MaterialExpressionSubtract_0, output=""}], properties {}}
  MaterialExpressionStaticSwitchParameter_5 {
   class=MaterialExpressionStaticSwitchParameter
   inputs [
    True {from=MaterialExpressionSetMaterialAttributes_3, output=""}
    False {from=MaterialExpressionSetMaterialAttributes_0, output=""}
   ]
   properties {group=Eyelid DF Lines Masking, parameterName=UseEyelidDFLinesMaskingDebug, sortPriority=-1}
  }
  MaterialExpressionSubtract_0 {
   class=MaterialExpressionSubtract
   inputs [
    A {from=MaterialExpressionComponentMask_0}
    B {from=""}
   ]
   properties {constB=0.0001}
  }
  MaterialExpressionTextureCoordinate_1 {class=MaterialExpressionTextureCoordinate, properties {coordinateIndex=1}}
  MaterialExpressionVertexInterpolator_0 {class=MaterialExpressionVertexInterpolator, inputs [VS {from=MaterialExpressionPreSkinnedPosition_0, output=""}], properties {}}
 }
 properties {
  bAutomaticallySetUsageInEditor=false
  bEnableResponsiveAA=true
  bUseMaterialAttributes=true
  bUsedWithMeshDeformer=true
  bUsedWithMorphTargets=true
  bUsedWithSkeletalMesh=true
  blendMode=BLEND_Masked
  pixelDepthOffsetMode=PDOM_Legacy
  shadingModel=MSM_Unlit
  twoSided=true
  usedShadingModels=Unlit
 }
}
