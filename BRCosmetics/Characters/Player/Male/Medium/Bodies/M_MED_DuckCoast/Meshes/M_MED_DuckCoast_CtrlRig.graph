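# Control Rig for the DuckCoast male medium player body, applying procedural corrective motion to skirt, foot and arm deform bones.
# A swing/twist DeltaRotation function remaps joint rotations into weights, while spherical pose readers drive upper-arm correctives.
# These weights, anim attributes and gameplay states like crouch, movement, DBNO, parachute and emotes blend additive bone offsets, often smoothed with AlphaInterp.
# Parent constraints blend skirt and foot bases between neighboring bones.
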ControlRigBlueprint /BRCosmetics/Characters/Player/Male/Medium/Bodies/M_MED_DuckCoast/Meshes/M_MED_DuckCoast_CtrlRig
graphs [
 Multiply_20_ContainedGraph {nodes [
   n1 = RigVMFunctionEntryNode()
   RigVMFunctionReturnNode(Result=&RigVMFunction_MathDoubleMul(B=@n1.C, A=&RigVMFunction_MathDoubleMul(A=@n1.A, B=@n1.B).Result).Result)
  ]}
 New Function_1_ContainedGraph {nodes [
   RigVMFunctionEntryNode()
   > n1 = &RigUnit_SphericalPoseReader(
    ActiveRegionScaleFactors=(PositiveWidth=1,NegativeWidth=1,PositiveHeight=1,NegativeHeight=1)
    ActiveRegionSize=0.08
    Debug=(bDrawDebug=false,bDraw2D=False,bDrawLocalAxes=False,DebugScale=25,DebugSegments=20,DebugThickness=0.25)
    DriverAxis=(X=1,Y=0,Z=0)
    DriverItem=(Type=Bone,Name="upperarm_l")
    FalloffRegionScaleFactors=(PositiveWidth=1,NegativeWidth=1,PositiveHeight=1,NegativeHeight=1)
    FalloffSize=0.2
    FlipHeightScaling=False
    FlipWidthScaling=False
    OptionalParentItem=(Type=Bone,Name="None")
    RotationOffset=(X=130,Y=0,Z=-80)
   )
   > &RigUnit_ModifyTransforms(
    ItemToModify="((Item=(Type=Bone,Name=\"upperarm_l_low_base\"),Transform=(Rotation=(X=0,Y=-0.0436,Z=0,W=0.999),Translation=(X=0,Y=0,Z=-1),Scale3D=(X=1,Y=1,Z=1))),(Item=(Type=Bone,Name=\"upperarm_twist_02_l\"),Transform=(Rotation=(X=0,Y=0,Z=0,W=1),Translation=(X=0,Y=0,Z=1.5), ...(truncated)"
    WeightMinimum=0
    Weight=@n1.OutputParam
   )
   > n2 = &RigUnit_SphericalPoseReader(
    ActiveRegionScaleFactors=(PositiveWidth=1,NegativeWidth=1,PositiveHeight=1,NegativeHeight=1)
    ActiveRegionSize=0.08
    Debug=(bDrawDebug=false,bDraw2D=False,bDrawLocalAxes=False,DebugScale=25,DebugSegments=20,DebugThickness=0.25)
    DriverAxis=(X=-1,Y=0,Z=0)
    DriverItem=(Type=Bone,Name="upperarm_r")
    FalloffRegionScaleFactors=(PositiveWidth=1,NegativeWidth=1,PositiveHeight=1,NegativeHeight=1)
    FalloffSize=0.2
    FlipHeightScaling=False
    FlipWidthScaling=False
    OptionalParentItem=(Type=Bone,Name="None")
    RotationOffset=(X=-50,Y=0,Z=-80)
   )
   > &RigUnit_ModifyTransforms(
    ItemToModify="((Item=(Type=Bone,Name=\"upperarm_r_low_base\"),Transform=(Rotation=(X=0,Y=-0.0349,Z=0,W=0.9994),Translation=(X=0,Y=0,Z=-0.9),Scale3D=(X=1,Y=1,Z=1))),(Item=(Type=Bone,Name=\"upperarm_twist_02_r\"),Transform=(Rotation=(X=0,Y=0,Z=0,W=1),Translation=(X=0,Y=0,Z=-1 ...(truncated)"
    WeightMinimum=0
    Weight=@n2.OutputParam
   )
   > RigVMFunctionReturnNode()
  ]}
 New Function_ContainedGraph {nodes [
   n1 = RigVMFunctionEntryNode()
   RigVMFunction_MathBoolToInteger = &RigVMFunction_MathBoolToInteger(Value=@n1.onlyTwist)
   RigVMFunction_MathQuaternionSwingTwist = &RigVMFunction_MathQuaternionSwingTwist(
    TwistAxis=@n1.twistAxis
    Input=&RigUnit_GetRelativeTransformForItem(bChildInitial=False, bParentInitial=False, Child=@n1.child, Parent=@n1.parent).RelativeTransform.Rotation
   )
   n2 = RigVMDispatchNode(
    Values=((),())
    Index=@RigVMFunction_MathBoolToInteger.Result
    Values.0=@RigVMFunction_MathQuaternionSwingTwist.Swing
    Values.1=@RigVMFunction_MathQuaternionSwingTwist.Twist
   ) {template="DISPATCH_RigVMDispatch_SelectInt32(in Index,in Values,out Result)"}
   RigVMFunction_MathQuaternionToEuler = &RigVMFunction_MathQuaternionToEuler(Value=@n2.Result, RotationOrder=@n1.Rotation Order)
   > DISPATCH_RigVMDispatch_Print = RigVMDispatchNode(ScreenColor=(R=1,G=1,B=1,A=1), ScreenDuration=1.5, Enabled=@n1.debugRotation, Value=@RigVMFunction_MathQuaternionToEuler.Result) {template="DISPATCH_RigVMDispatch_Print(in Prefix,in Value,in Enabled,in ScreenDuration,in ScreenColor)"}
   Cast = RigVMDispatchNode(Value=@n1.triggerAxis) {template="DISPATCH_RigVMDispatch_CastEnumToInt(in Value,out Result)"}
   DISPATCH_RigVMDispatch_SelectInt32 = RigVMDispatchNode(
    Values=(0,0,0)
    Index=@Cast.Result
    Values.0=@RigVMFunction_MathQuaternionToEuler.Result.X
    Values.1=@RigVMFunction_MathQuaternionToEuler.Result.Y
    Values.2=@RigVMFunction_MathQuaternionToEuler.Result.Z
   ) {template="DISPATCH_RigVMDispatch_SelectInt32(in Index,in Values,out Result)"}
   Remap = &RigVMFunction_MathDoubleRemap(
    bClamp=False
    Value=@DISPATCH_RigVMDispatch_SelectInt32.Result
    TargetMinimum=@n1.outputMinMax.X
    TargetMaximum=@n1.outputMinMax.Y
    SourceMinimum=@n1.rotationMinMax.X
    SourceMaximum=@n1.rotationMinMax.Y
   )
   Absolute = &RigVMFunction_MathDoubleAbs(Value=@Remap.Result)
   Multiply = &RigVMFunction_MathDoubleMul(
    A=&RigVMFunction_MathDoubleClamp(Value=RigVMDispatchNode(True=@Absolute.Result, Condition=@n1.absoluteValue, False=@Remap.Result) {template="DISPATCH_RigVMDispatch_If(in Condition,in True,in False,out Result)"}.Result).Result
    B=@n1.premult
   )
   > RigVMFunctionReturnNode(outputQuat=@n2.Result, outputWeight=@Multiply.Result)
  ]}
 New Function_ContainedGraph {nodes [
   n1 = RigVMFunctionEntryNode()
   RigVMFunction_ControlFlowBranch = &RigVMFunction_ControlFlowBranch(Condition=@n1.Debug Controls, ExecuteContext=@n1.ExecuteContext)
   RigVMFunctionReturnNode(ExecuteContext=@RigVMFunction_ControlFlowBranch.Completed)
   RigUnit_GetTransform = &RigUnit_GetTransform(Item=(Type=Control,Name="thigh_l_ctrl"), bInitial=False)
   &RigUnit_SetTransform(
    Item=(Type=Bone,Name="thigh_l")
    Weight=1
    bInitial=False
    Value=@RigUnit_GetTransform.Transform
    ExecutePin=@RigVMFunction_ControlFlowBranch.True
   )
   > &RigUnit_SetTransform(
    Item=(Type=Bone,Name="thigh_r")
    Weight=1
    bInitial=False
    Value=&RigUnit_GetTransform(Item=(Type=Control,Name="thigh_r_ctrl"), bInitial=False).Transform
   )
   > &RigUnit_SetTransform(
    Item=(Type=Bone,Name="calf_l")
    Weight=1
    bInitial=False
    Value=&RigUnit_GetTransform(Item=(Type=Control,Name="calf_l_ctrl"), bInitial=False).Transform
   )
   > &RigUnit_SetTransform(
    Item=(Type=Bone,Name="calf_r")
    Weight=1
    bInitial=False
    Value=&RigUnit_GetTransform(Item=(Type=Control,Name="calf_r_ctrl"), bInitial=False).Transform
   )
  ]}
 New Function_ContainedGraph {nodes [
   n1 = RigVMFunctionEntryNode()
   RigVMFunction_Sequence = &RigVMFunction_Sequence(ExecuteContext=@n1.ExecuteContext)
   DeltaRotation v4 = RigVMFunctionReferenceNode(
    Rotation Order=ZYX
    child=(Type=Bone,Name="thigh_l")
    debugRotation=false
    outputMinMax=(X=0,Y=1)
    parent=(Type=Null,Name="thigh_l_null")
    premult=1
    rotationMinMax=(X=0,Y=-120)
    triggerAxis=Z
    twistAxis=(X=1,Y=0,Z=0)
    ExecuteContext=@RigVMFunction_Sequence.A
   )
   > RigVMVariableNode(Value=@"DeltaRotation v4".outputWeight) {variable=SPR L Leg Fwd}
   > DeltaRotation v4_1 = RigVMFunctionReferenceNode(
    Rotation Order=ZYX
    child=(Type=Bone,Name="thigh_r")
    debugRotation=false
    outputMinMax=(X=0,Y=1)
    parent=(Type=Null,Name="thigh_r_null")
    premult=1
    rotationMinMax=(X=0,Y=-120)
    triggerAxis=Z
    twistAxis=(X=1,Y=0,Z=0)
   )
   > RigVMVariableNode(Value=@"DeltaRotation v4_1".outputWeight) {variable=SPR R Leg Fwd}
   > DeltaRotation v4_2 = RigVMFunctionReferenceNode(
    Rotation Order=ZYX
    child=(Type=Bone,Name="thigh_l")
    debugRotation=false
    outputMinMax=(X=0,Y=1)
    parent=(Type=Null,Name="thigh_l_null")
    premult=1
    rotationMinMax=(X=0,Y=90)
    triggerAxis=Z
    twistAxis=(X=1,Y=0,Z=0)
   )
   > RigVMVariableNode(Value=@"DeltaRotation v4_2".outputWeight) {variable=SPR L Leg Bkd}
   > DeltaRotation v4_3 = RigVMFunctionReferenceNode(
    Rotation Order=ZYX
    child=(Type=Bone,Name="thigh_r")
    debugRotation=false
    outputMinMax=(X=0,Y=1)
    parent=(Type=Null,Name="thigh_r_null")
    premult=1
    rotationMinMax=(X=0,Y=90)
    triggerAxis=Z
    twistAxis=(X=1,Y=0,Z=0)
   )
   > RigVMVariableNode(Value=@"DeltaRotation v4_3".outputWeight) {variable=SPR R Leg Bkd}
   > DeltaRotation v4_4 = RigVMFunctionReferenceNode(
    Rotation Order=ZYX
    child=(Type=Bone,Name="thigh_l")
    debugRotation=false
    outputMinMax=(X=0,Y=1)
    parent=(Type=Null,Name="thigh_l_null")
    premult=1
    rotationMinMax=(X=0,Y=-90)
    triggerAxis=Y
    twistAxis=(X=1,Y=0,Z=0)
   )
   > RigVMVariableNode(Value=@"DeltaRotation v4_4".outputWeight) {variable=SPR L Leg Otw}
   > DeltaRotation v4_5 = RigVMFunctionReferenceNode(
    Rotation Order=ZYX
    child=(Type=Bone,Name="thigh_r")
    debugRotation=false
    outputMinMax=(X=0,Y=1)
    parent=(Type=Null,Name="thigh_r_null")
    premult=1
    rotationMinMax=(X=0,Y=-90)
    triggerAxis=Y
    twistAxis=(X=1,Y=0,Z=0)
   )
   > RigVMVariableNode(Value=@"DeltaRotation v4_5".outputWeight) {variable=SPR R Leg Otw}
   > DeltaRotation v4_6 = RigVMFunctionReferenceNode(
    Rotation Order=ZYX
    child=(Type=Bone,Name="calf_l")
    debugRotation=false
    outputMinMax=(X=0,Y=1)
    parent=(Type=Null,Name="calf_l_null")
    premult=1
    rotationMinMax=(X=40,Y=120)
    triggerAxis=Z
    twistAxis=(X=1,Y=0,Z=0)
   )
   > RigVMVariableNode(Value=@"DeltaRotation v4_6".outputWeight) {variable=SPR L Calf Bkd}
   > DeltaRotation v4_7 = RigVMFunctionReferenceNode(
    Rotation Order=ZYX
    child=(Type=Bone,Name="calf_l")
    debugRotation=false
    outputMinMax=(X=0,Y=1)
    parent=(Type=Null,Name="calf_l_null")
    premult=1
    rotationMinMax=(X=120,Y=130)
    triggerAxis=Z
    twistAxis=(X=1,Y=0,Z=0)
   )
   > RigVMVariableNode(Value=@"DeltaRotation v4_7".outputWeight) {variable=SPR L Ext Calf Bkd}
   > DeltaRotation v4_8 = RigVMFunctionReferenceNode(
    Rotation Order=ZYX
    child=(Type=Bone,Name="calf_r")
    debugRotation=false
    outputMinMax=(X=0,Y=1)
    parent=(Type=Null,Name="calf_r_null")
    premult=1
    rotationMinMax=(X=50,Y=120)
    triggerAxis=Z
    twistAxis=(X=1,Y=0,Z=0)
   )
   > RigVMVariableNode(Value=@"DeltaRotation v4_8".outputWeight) {variable=SPR R Calf Bkd}
   > DeltaRotation v4_9 = RigVMFunctionReferenceNode(
    Rotation Order=ZYX
    child=(Type=Bone,Name="calf_r")
    debugRotation=false
    outputMinMax=(X=0,Y=1)
    parent=(Type=Null,Name="calf_r_null")
    premult=1
    rotationMinMax=(X=120,Y=130)
    triggerAxis=Z
    twistAxis=(X=1,Y=0,Z=0)
   )
   > RigVMVariableNode(Value=@"DeltaRotation v4_9".outputWeight) {variable=SPR R Ext Calf Bkd}
   > DeltaRotation v4_10 = RigVMFunctionReferenceNode(
    Rotation Order=ZYX
    child=(Type=Bone,Name="thigh_l")
    debugRotation=false
    onlyTwist=false
    outputMinMax=(X=0,Y=1)
    parent=(Type=Null,Name="thigh_l_null")
    premult=1
    rotationMinMax=(X=0,Y=-90)
    triggerAxis=Z
    twistAxis=(X=1,Y=0,Z=0)
   )
   > DeltaRotation v4_11 = RigVMFunctionReferenceNode(
    Rotation Order=ZYX
    child=(Type=Bone,Name="spine_05")
    debugRotation=false
    outputMinMax=(X=0,Y=1)
    parent=(Type=Null,Name="spine_05_null")
    premult=1
    rotationMinMax=(X=0,Y=8)
    triggerAxis=Z
    twistAxis=(X=1,Y=0,Z=0)
   )
   > DeltaRotation v4_12 = RigVMFunctionReferenceNode(
    Rotation Order=ZYX
    child=(Type=Bone,Name="foot_l")
    debugRotation=false
    outputMinMax=(X=0,Y=1)
    parent=(Type=Null,Name="foot_l_null")
    premult=1
    rotationMinMax=(X=0,Y=65)
    triggerAxis=Z
    twistAxis=(X=1,Y=0,Z=0)
   )
   > RigVMVariableNode(Value=@"DeltaRotation v4_12".outputWeight) {variable=SPR L Foot}
   > DeltaRotation v4_13 = RigVMFunctionReferenceNode(
    Rotation Order=ZYX
    child=(Type=Bone,Name="foot_r")
    debugRotation=false
    outputMinMax=(X=0,Y=1)
    parent=(Type=Null,Name="foot_r_null")
    premult=1
    rotationMinMax=(X=0,Y=65)
    triggerAxis=Z
    twistAxis=(X=1,Y=0,Z=0)
   )
   > RigVMVariableNode(Value=@"DeltaRotation v4_13".outputWeight) {variable=SPR R Foot}
   > DeltaRotation v4_14 = RigVMFunctionReferenceNode(
    Rotation Order=ZYX
    child=(Type=Bone,Name="calf_l")
    debugRotation=false
    outputMinMax=(X=0,Y=1)
    parent=(Type=Null,Name="calf_l_null")
    premult=1
    rotationMinMax=(X=0,Y=25)
    triggerAxis=Z
    twistAxis=(X=1,Y=0,Z=0)
   )
   > RigVMVariableNode(Value=@"DeltaRotation v4_14".outputWeight) {variable=SPR L Calf Fixes}
   > DeltaRotation v4_15 = RigVMFunctionReferenceNode(
    Rotation Order=ZYX
    child=(Type=Bone,Name="calf_r")
    debugRotation=false
    outputMinMax=(X=0,Y=1)
    parent=(Type=Null,Name="calf_r_null")
    premult=1
    rotationMinMax=(X=0,Y=10)
    triggerAxis=Z
    twistAxis=(X=1,Y=0,Z=0)
   )
   > RigVMVariableNode(Value=@"DeltaRotation v4_15".outputWeight) {variable=SPR R Calf Fixes}
   > DeltaRotation v4_16 = RigVMFunctionReferenceNode(
    Rotation Order=ZYX
    child=(Type=Bone,Name="thigh_l")
    debugRotation=false
    outputMinMax=(X=0,Y=1)
    parent=(Type=Null,Name="thigh_l_null")
    premult=1
    rotationMinMax=(X=0,Y=-45)
    triggerAxis=Z
    twistAxis=(X=1,Y=0,Z=0)
   )
   > RigVMVariableNode(Value=@"DeltaRotation v4_16".outputWeight) {variable=SPR L Thigh Fixes}
   > DeltaRotation v4_17 = RigVMFunctionReferenceNode(
    Rotation Order=ZYX
    child=(Type=Bone,Name="thigh_r")
    debugRotation=false
    outputMinMax=(X=0,Y=1)
    parent=(Type=Null,Name="thigh_r_null")
    premult=1
    rotationMinMax=(X=0,Y=-45)
    triggerAxis=Z
    twistAxis=(X=1,Y=0,Z=0)
   )
   > RigVMVariableNode(Value=@"DeltaRotation v4_17".outputWeight) {variable=SPR R Thigh Fixes}
   RigVMFunctionReturnNode(ExecuteContext=@RigVMFunction_Sequence.B)
  ]}
 New Function_ContainedGraph {nodes [
   RigVMFunctionEntryNode()
   > RigVMFunction_Sequence = RigVMAggregateNode()
   Greater = /Script/RigVM.RigVMFunction_MathDoubleGreater(B=0.1, A=RigVMVariableNode() {variable=Is_on_Air}.Value) {is=struct RigVMUnitNode, template="Greater::Execute(in A,in B,out Result)"}
   RigVMFunction_ControlFlowBranch = &RigVMFunction_ControlFlowBranch(Condition=@Greater.Result, ExecuteContext=@RigVMFunction_Sequence.B)
   n1 = &RigVMFunction_MathDoubleRemap(
    SourceMaximum=0
    SourceMinimum=1
    TargetMaximum=1
    TargetMinimum=0
    bClamp=False
    Value=RigVMVariableNode() {variable=Is_BreakDance_Emote}.Value
   )
   &RigUnit_ParentConstraint(
    AdvancedSettings=(RotationOrderForFilter=XZY)
    Child=(Type=Bone,Name="foot_l_base_deform")
    Parents=((Item=(Type=Bone,Name="calf_twist_02_l"),Weight=0.006),(Item=(Type=Bone,Name="calf_twist_01_l"),Weight=0.2703),(Item=(Type=Bone,Name="foot_l"),Weight=0.7237))
    Weight=@n1.Result
    ExecutePin=@RigVMFunction_ControlFlowBranch.False
   )
   > &RigUnit_ParentConstraint(
    AdvancedSettings=(RotationOrderForFilter=XZY)
    Child=(Type=Bone,Name="foot_r_base_deform")
    Parents=((Item=(Type=Bone,Name="calf_twist_02_r"),Weight=0.006),(Item=(Type=Bone,Name="calf_twist_01_r"),Weight=0.2703),(Item=(Type=Bone,Name="foot_r"),Weight=0.7237))
    Weight=@n1.Result
   )
   > &RigUnit_ParentConstraint(
    AdvancedSettings=(InterpolationType=Shortest,RotationOrderForFilter=XZY)
    Child=(Type=Bone,Name="SkirtB_01_Base")
    Filter=(TranslationFilter=(bX=true,bY=true,bZ=true),RotationFilter=(bX=true,bY=true,bZ=true),ScaleFilter=(bX=false,bY=false,bZ=false))
    Parents=((Item=(Type=Bone,Name="SkirtC_l_01_dyn"),Weight=0.3),(Item=(Type=Bone,Name="SkirtC_r_01_dyn"),Weight=0.3),(Item=(Type=Bone,Name="pelvis"),Weight=0.4))
    Weight=@n1.Result
   )
   > &RigUnit_ParentConstraint(
    AdvancedSettings=(InterpolationType=Shortest,RotationOrderForFilter=XZY)
    Child=(Type=Bone,Name="SkirtB_02_dyn")
    Filter=(TranslationFilter=(bX=true,bY=true,bZ=true),RotationFilter=(bX=true,bY=true,bZ=true),ScaleFilter=(bX=false,bY=false,bZ=false))
    Parents=((Item=(Type=Bone,Name="SkirtC_l_02_dyn"),Weight=0.3),(Item=(Type=Bone,Name="SkirtC_r_02_dyn"),Weight=0.3),(Item=(Type=Bone,Name="pelvis"),Weight=0.4))
    Weight=@n1.Result
   )
   > &RigUnit_ParentConstraint(
    AdvancedSettings=(InterpolationType=Shortest,RotationOrderForFilter=XZY)
    Child=(Type=Bone,Name="SkirtB_03_dyn")
    Filter=(TranslationFilter=(bX=true,bY=true,bZ=true),RotationFilter=(bX=true,bY=true,bZ=true),ScaleFilter=(bX=false,bY=false,bZ=false))
    Parents=((Item=(Type=Bone,Name="SkirtC_l_03_dyn"),Weight=0.3),(Item=(Type=Bone,Name="SkirtC_r_03_dyn"),Weight=0.3),(Item=(Type=Bone,Name="pelvis"),Weight=0.4))
    bMaintainOffset=true
    Weight=@n1.Result
   )
   > &RigUnit_ParentConstraint(
    AdvancedSettings=(InterpolationType=Shortest,RotationOrderForFilter=XZY)
    Child=(Type=Bone,Name="SkirtC_l_01_Base")
    Filter=(TranslationFilter=(bX=false,bY=true,bZ=true),RotationFilter=(bX=true,bY=true,bZ=false),ScaleFilter=(bX=false,bY=false,bZ=false))
    Parents=((Item=(Type=Bone,Name="thigh_l"),Weight=0.25),(Item=(Type=Bone,Name="pelvis"),Weight=0.75))
    Weight=@n1.Result
   )
   > &RigUnit_ParentConstraint(
    AdvancedSettings=(InterpolationType=Shortest,RotationOrderForFilter=XZY)
    Child=(Type=Bone,Name="SkirtC_r_01_Base")
    Filter=(TranslationFilter=(bX=false,bY=true,bZ=true),RotationFilter=(bX=true,bY=true,bZ=false),ScaleFilter=(bX=false,bY=false,bZ=false))
    Parents=((Item=(Type=Bone,Name="thigh_r"),Weight=0.25),(Item=(Type=Bone,Name="pelvis"),Weight=0.75))
    Weight=@n1.Result
   )
   n2 = RigVMVariableNode() {variable=SPR L Leg Fwd}
   > &RigUnit_ModifyTransforms(
    ItemToModify="((Item=(Type=Bone,Name=\"SkirtA_l_01_Base\"),Transform=(Rotation=(X=0,Y=0.0436,Z=0,W=0.999),Translation=(X=0,Y=0,Z=0),Scale3D=(X=1,Y=1,Z=1))),(Item=(Type=Bone,Name=\"SkirtF_01_Base\"),Transform=(Rotation=(X=0,Y=-0.9816,Z=0,W=0.1908),Translation=(X=-6,Y=0,Z=0), ...(truncated)"
    WeightMinimum=0
    Weight=@n2.Value
   )
   n3 = RigVMVariableNode() {variable=SPR R Leg Fwd}
   > &RigUnit_ModifyTransforms(
    ItemToModify="((Item=(Type=Bone,Name=\"SkirtA_r_01_Base\"),Transform=(Rotation=(X=0,Y=0.0436,Z=0,W=0.999),Translation=(X=0,Y=0,Z=0),Scale3D=(X=1,Y=1,Z=1))),(Item=(Type=Bone,Name=\"SkirtF_01_Base\"),Transform=(Rotation=(X=0,Y=-0.9816,Z=0,W=0.1908),Translation=(X=-6,Y=0,Z=0), ...(truncated)"
    WeightMinimum=0
    Weight=@n3.Value
   )
   Multiply = &RigVMFunction_MathDoubleMul(B=@n2.Value, A=@n3.Value)
   > &RigUnit_ModifyTransforms(
    ItemToModify=((Item=(Type=Bone,Name="SkirtF_01_Base"),Transform=(Rotation=(X=0,Y=1,Z=0,W=0),Translation=(X=-10,Y=0,Z=0),Scale3D=(X=1,Y=1,Z=1))))
    WeightMinimum=0
    Weight=@Multiply.Result
   )
   n4 = RigVMVariableNode() {variable=SPR L Leg Bkd}
   > &RigUnit_ModifyTransforms(
    ItemToModify="((Item=(Type=Bone,Name=\"SkirtC_l_01_Base\"),Transform=(Rotation=(X=-0.0021,Y=0.0053,Z=-0.102,W=-0.9948),Translation=(X=6.0002,Y=0,Z=-5.0001),Scale3D=(X=1,Y=1,Z=1))),(Item=(Type=Bone,Name=\"SkirtB_l_01_Base\"),Transform=(Rotation=(X=0,Y=0,Z=0,W=1),Translation= ...(truncated)"
    WeightMinimum=0
    Weight=@n4.Value
   )
   n5 = RigVMVariableNode() {variable=SPR R Leg Bkd}
   > &RigUnit_ModifyTransforms(
    ItemToModify="((Item=(Type=Bone,Name=\"SkirtC_r_01_Base\"),Transform=(Rotation=(X=0.0021,Y=-0.0053,Z=0.102,W=0.9948),Translation=(X=-6,Y=0,Z=5),Scale3D=(X=1,Y=1,Z=1))),(Item=(Type=Bone,Name=\"SkirtB_r_01_Base\"),Transform=(Rotation=(X=0,Y=0,Z=0.342,W=0.9397),Translation=(X= ...(truncated)"
    WeightMinimum=0
    Weight=@n5.Value
   )
   > &RigUnit_ModifyTransforms(
    ItemToModify=((Item=(Type=Bone,Name="SkirtB_01_Base"),Transform=(Rotation=(X=0,Y=0.6671,Z=0,W=0.745),Translation=(X=-16.7572,Y=0,Z=8.8711),Scale3D=(X=1,Y=1,Z=1))))
    WeightMinimum=0
    Weight=&RigVMFunction_MathDoubleMul(B=@n4.Value, A=@n5.Value).Result
   )
   n6 = RigVMVariableNode() {variable=SPR L Calf Bkd}
   > &RigUnit_ModifyTransforms(
    ItemToModify="((Item=(Type=Bone,Name=\"None\"),Transform=(Rotation=(X=0,Y=0,Z=0,W=1),Translation=(X=0,Y=0,Z=0),Scale3D=(X=1,Y=1,Z=1))),(Item=(Type=Bone,Name=\"SkirtC_l_02_dyn\"),Transform=(Rotation=(X=-0.1271,Y=-0.2725,Z=-0.4031,W=0.8644),Translation=(X=-10,Y=0,Z=0),Scale3D ...(truncated)"
    WeightMinimum=0
    Weight=@n6.Value
   )
   n7 = RigVMVariableNode() {variable=SPR R Calf Bkd}
   > &RigUnit_ModifyTransforms(
    ItemToModify="((Item=(Type=Bone,Name=\"None\"),Transform=(Rotation=(X=0,Y=0,Z=0,W=1),Translation=(X=0,Y=0,Z=0),Scale3D=(X=1,Y=1,Z=1))),(Item=(Type=Bone,Name=\"SkirtC_r_02_dyn\"),Transform=(Rotation=(X=-0.3044,Y=-0.5272,Z=-0.3967,W=0.6871),Translation=(X=5,Y=0,Z=0),Scale3D=( ...(truncated)"
    WeightMinimum=0
    Weight=@n7.Value
   )
   > &RigUnit_ModifyTransforms(
    ItemToModify="((Item=(Type=Bone,Name=\"SkirtB_02_dyn\"),Transform=(Rotation=(X=0,Y=-0.0872,Z=0,W=0.9962),Translation=(X=0,Y=0,Z=0),Scale3D=(X=1,Y=1,Z=1))),(Item=(Type=Bone,Name=\"SkirtB_03_dyn\"),Transform=(Rotation=(X=0,Y=0.4226,Z=0,W=0.9063),Translation=(X=0,Y=0,Z=0),Scal ...(truncated)"
    WeightMinimum=0
    Weight=&RigVMFunction_MathDoubleMul(B=@n6.Value, A=@n7.Value).Result
   )
   n8 = RigVMVariableNode() {variable=SPR L Ext Calf Bkd}
   > &RigUnit_ModifyTransforms(
    ItemToModify="((Item=(Type=Bone,Name=\"SkirtC_l_01_Base\"),Transform=(Rotation=(X=0,Y=0,Z=0,W=1),Translation=(X=-8,Y=0,Z=0),Scale3D=(X=1,Y=1,Z=1))),(Item=(Type=Bone,Name=\"SkirtC_l_02_dyn\"),Transform=(Rotation=(X=0,Y=0,Z=0,W=1),Translation=(X=0,Y=0,Z=0),Scale3D=(X=1,Y=1,Z= ...(truncated)"
    WeightMinimum=0
    Weight=@n8.Value
   )
   n9 = RigVMVariableNode() {variable=SPR R Ext Calf Bkd}
   > &RigUnit_ModifyTransforms(
    ItemToModify="((Item=(Type=Bone,Name=\"None\"),Transform=(Rotation=(X=0,Y=0,Z=0,W=1),Translation=(X=0,Y=0,Z=0),Scale3D=(X=1,Y=1,Z=1))),(Item=(Type=Bone,Name=\"SkirtC_r_01_Base\"),Transform=(Rotation=(X=0,Y=0,Z=0,W=-1),Translation=(X=12.5813,Y=0,Z=-2.2772),Scale3D=(X=1,Y=1,Z ...(truncated)"
    WeightMinimum=0
    Weight=@n9.Value
   )
   > &RigUnit_ModifyTransforms(
    ItemToModify="((Item=(Type=Bone,Name=\"SkirtB_02_dyn\"),Transform=(Rotation=(X=0,Y=0,Z=0,W=1),Translation=(X=0,Y=0,Z=0),Scale3D=(X=1,Y=1,Z=1))),(Item=(Type=Bone,Name=\"SkirtB_03_dyn\"),Transform=(Rotation=(X=-0.1556,Y=0.0457,Z=-0.0095,W=0.9867),Translation=(X=0,Y=0,Z=0),Sca ...(truncated)"
    WeightMinimum=0
    Weight=&RigVMFunction_MathDoubleMul(B=@n8.Value, A=@n9.Value).Result
   )
   > &RigUnit_ModifyTransforms(
    ItemToModify="((Item=(Type=Bone,Name=\"SkirtC_r_01_dyn\"),Transform=(Rotation=(X=-0.0106,Y=-0.121,Z=0.0865,W=0.9888),Translation=(X=13.8978,Y=0.1874,Z=-7.5061),Scale3D=(X=1,Y=1,Z=1))),(Item=(Type=Bone,Name=\"SkirtC_r_02_dyn\"),Transform=(Rotation=(X=0,Y=0,Z=-0.0872,W=0.9962 ...(truncated)"
    WeightMinimum=0
    Weight=RigVMVariableNode() {variable=Is_Crouching}.Value
   )
   n10 = &RigVMFunction_AlphaInterp(
    InterpSpeedDecreasing=5
    InterpSpeedIncreasing=30
    Scale=1
    bClampResult=False
    Value=&RigVMFunction_MathDoubleMul(A=RigVMVariableNode() {variable=Is_Crouching}.Value, B=RigVMVariableNode() {variable=is_Moving_Forward}.Value).Result
   )
   > &RigUnit_ModifyTransforms(
    ItemToModify="((Item=(Type=Bone,Name=\"SkirtC_l_02_dyn\"),Transform=(Rotation=(X=-0.0547,Y=0.3219,Z=-0.1582,W=0.9319),Translation=(X=-3.2531,Y=0.5238,Z=2.552),Scale3D=(X=1,Y=1,Z=1))),(Item=(Type=Bone,Name=\"SkirtC_r_03_dyn\"),Transform=(Rotation=(X=0,Y=-0.1575,Z=0,W=0.9875) ...(truncated)"
    WeightMinimum=0
    Weight=@n10.Result
   )
   n11 = &RigVMFunction_MathDoubleMul(A=RigVMVariableNode() {variable=is_Moving_Forward}.Value, B=RigVMVariableNode() {variable=is_Moving_left}.Value)
   > &RigUnit_ModifyTransforms(
    ItemToModify="((Item=(Type=Bone,Name=\"SkirtB_01_dyn\"),Transform=(Rotation=(X=-0.0041,Y=-0.0277,Z=-0.4891,W=0.8718),Translation=(X=-2.2942,Y=-0.6154,Z=-0.0478),Scale3D=(X=1,Y=1,Z=1))),(Item=(Type=Bone,Name=\"SkirtC_l_01_dyn\"),Transform=(Rotation=(X=0,Y=0,Z=0,W=1),Translat ...(truncated)"
    WeightMinimum=0
    Weight=@n11.Result
   )
   Subtract = &RigVMFunction_MathDoubleSub(
    B=@n11.Result
    A=&RigVMFunction_MathDoubleMul(A=RigVMVariableNode() {variable=Is_Crouching}.Value, B=RigVMVariableNode() {variable=is_Moving_left}.Value).Result
   )
   > &RigUnit_ModifyTransforms(
    ItemToModify="((Item=(Type=Bone,Name=\"SkirtB_l_01_dyn\"),Transform=(Rotation=(X=0,Y=0.1947,Z=0,W=-0.9809),Translation=(X=-5.4182,Y=0.0001,Z=-2.2391),Scale3D=(X=1,Y=1,Z=1))),(Item=(Type=Bone,Name=\"SkirtB_l_02_dyn\"),Transform=(Rotation=(X=-0.0011,Y=-0.0217,Z=-0.049,W=-0.99 ...(truncated)"
    WeightMinimum=0
    Weight=@Subtract.Result
   )
   n12 = &RigVMFunction_AlphaInterp(
    InterpSpeedDecreasing=5
    InterpSpeedIncreasing=5
    Scale=1
    bClampResult=False
    Value=&RigVMFunction_MathDoubleMul(A=RigVMVariableNode() {variable=is_Moving_Back}.Value, B=RigVMVariableNode() {variable=Is_Crouching}.Value).Result
   )
   > &RigUnit_ModifyTransforms(
    ItemToModify="((Item=(Type=Bone,Name=\"SkirtC_l_02_dyn\"),Transform=(Rotation=(X=0.3794,Y=0.05,Z=0.1206,W=0.916),Translation=(X=0,Y=0,Z=0),Scale3D=(X=1,Y=1,Z=1))),(Item=(Type=Bone,Name=\"SkirtB_02_dyn\"),Transform=(Rotation=(X=0,Y=0.2506,Z=0,W=0.9681),Translation=(X=0,Y=0,Z ...(truncated)"
    WeightMinimum=0
    Weight=@n12.Result
   )
   > &RigUnit_ModifyTransforms(
    ItemToModify=((Item=(Type=Bone,Name="SkirtF_01_dyn"),Transform=(Rotation=(X=0.7237,Y=0.4962,Z=-0.1253,W=-0.4629),Translation=(X=0,Y=0,Z=0),Scale3D=(X=1,Y=1,Z=1))))
    WeightMinimum=0
    Weight=&RigVMFunction_MathDoubleMul(A=RigVMVariableNode() {variable=is_Moving_Forward}.Value, B=RigVMVariableNode() {variable=is_Moving_right}.Value).Result
   )
   > &RigUnit_ModifyTransforms(
    ItemToModify="((Item=(Type=Bone,Name=\"SkirtF_01_dyn\"),Transform=(Rotation=(X=0.0583,Y=0.1135,Z=-0.2825,W=-0.9507),Translation=(X=0,Y=0,Z=0),Scale3D=(X=1,Y=1,Z=1))),(Item=(Type=Bone,Name=\"SkirtC_l_02_dyn\"),Transform=(Rotation=(X=0.0013,Y=0.0331,Z=0.0403,W=0.9986),Transla ...(truncated)"
    WeightMinimum=0
    Weight=&RigVMFunction_MathDoubleMul(B=RigVMVariableNode() {variable=is_Moving_right}.Value, A=RigVMVariableNode() {variable=Is_Crouching}.Value).Result
   )
   n13 = &RigVMFunction_AlphaInterp(
    InterpSpeedDecreasing=10
    InterpSpeedIncreasing=5
    Scale=1
    bClampResult=False
    Value=&RigVMFunction_MathDoubleMul(A=RigVMVariableNode() {variable=is_Moving_Forward}.Value, B=RigVMVariableNode() {variable=Is not crouching}.Value).Result
   )
   > n14 = &RigUnit_ModifyTransforms(
    ItemToModify="((Item=(Type=Bone,Name=\"None\"),Transform=(Rotation=(X=0,Y=0,Z=0,W=1),Translation=(X=0,Y=0,Z=0),Scale3D=(X=1,Y=1,Z=1))),(Item=(Type=Bone,Name=\"SkirtC_l_01_dyn\"),Transform=(Rotation=(X=0,Y=0,Z=0,W=-1),Translation=(X=-5,Y=0,Z=3),Scale3D=(X=1,Y=1,Z=1))),(Item= ...(truncated)"
    WeightMinimum=0
    Weight=@n13.Result
   )
   n15 = RigVMVariableNode() {variable=DBNO Carried}
   n16 = &RigVMFunction_ControlFlowBranch(
    ExecuteContext=@n14.ExecutePin
    Condition=&RigVMFunction_MathBoolAnd(B=&RigVMFunction_MathBoolNot(Value=@n15.Value).Result, A=RigVMVariableNode() {variable=is DBNO}.Value).Result
   )
   &RigUnit_ModifyTransforms(
    ItemToModify="((Item=(Type=Bone,Name=\"SkirtA_l_01_dyn\"),Transform=(Rotation=(X=0,Y=-0.143,Z=0,W=-0.9897),Translation=(X=0,Y=0,Z=0),Scale3D=(X=1,Y=1,Z=1))),(Item=(Type=Bone,Name=\"SkirtA_r_01_dyn\"),Transform=(Rotation=(X=-0.0046,Y=0.148,Z=-0.031,W=0.9885),Translation=(X=0 ...(truncated)"
    WeightMinimum=0
    Weight=RigVMDispatchNode(False=0, True=1, Condition=RigVMVariableNode() {variable=is DBNO}.Value) {template="DISPATCH_RigVMDispatch_If(in Condition,in True,in False,out Result)"}.Result
    ExecutePin=@n16.True
   )
   n17 = RigVMVariableNode() {variable=is DBNO}
   n18 = &RigVMFunction_AlphaInterp(
    InterpSpeedDecreasing=4
    InterpSpeedIncreasing=10
    Scale=1
    bClampResult=False
    Value=&RigVMFunction_MathDoubleMul(
     A=RigVMDispatchNode(False=0, True=1, Condition=@n17.Value) {template="DISPATCH_RigVMDispatch_If(in Condition,in True,in False,out Result)"}.Result
     B=RigVMVariableNode() {variable=is_Moving_Forward}.Value
    ).Result
   )
   > &RigUnit_ModifyTransforms(
    ItemToModify="((Item=(Type=Bone,Name=\"SkirtC_r_02_dyn\"),Transform=(Rotation=(X=-0.4813,Y=-0.1769,Z=0.413,W=-0.7527),Translation=(X=0,Y=0,Z=0),Scale3D=(X=1,Y=1,Z=1))),(Item=(Type=Bone,Name=\"SkirtA_l_02_dyn\"),Transform=(Rotation=(X=0,Y=-0.1524,Z=0,W=0.9883),Translation=(X ...(truncated)"
    WeightMinimum=0
    Weight=@n18.Result
   )
   n19 = RigVMVariableNode() {variable=is DBNO}
   n20 = &RigVMFunction_AlphaInterp(
    InterpSpeedDecreasing=3
    InterpSpeedIncreasing=10
    Scale=1
    bClampResult=False
    Value=&RigVMFunction_MathDoubleMul(
     A=RigVMDispatchNode(False=0, True=1, Condition=@n19.Value) {template="DISPATCH_RigVMDispatch_If(in Condition,in True,in False,out Result)"}.Result
     B=RigVMVariableNode() {variable=is_Moving_right}.Value
    ).Result
   )
   > &RigUnit_ModifyTransforms(
    ItemToModify="((Item=(Type=Bone,Name=\"SkirtC_r_02_dyn\"),Transform=(Rotation=(X=0.0088,Y=0.1318,Z=0.4116,W=-0.9018),Translation=(X=7.7197,Y=-3.8131,Z=0.8403),Scale3D=(X=1,Y=1,Z=1))),(Item=(Type=Bone,Name=\"SkirtA_r_02_dyn\"),Transform=(Rotation=(X=0.0016,Y=-0.0727,Z=-0.022 ...(truncated)"
    WeightMinimum=0
    Weight=@n20.Result
   )
   n21 = RigVMVariableNode() {variable=is DBNO}
   n22 = &RigVMFunction_AlphaInterp(
    InterpSpeedDecreasing=0.4
    InterpSpeedIncreasing=10
    Scale=1
    bClampResult=False
    Value=&RigVMFunction_MathDoubleMul(
     A=RigVMDispatchNode(False=0, True=1, Condition=@n21.Value) {template="DISPATCH_RigVMDispatch_If(in Condition,in True,in False,out Result)"}.Result
     B=RigVMVariableNode() {variable=is_Moving_left}.Value
    ).Result
   )
   > &RigUnit_ModifyTransforms(
    ItemToModify="((Item=(Type=Bone,Name=\"SkirtC_r_02_dyn\"),Transform=(Rotation=(X=0,Y=0,Z=-0.4434,W=0.8963),Translation=(X=5.0619,Y=-4.2247,Z=0),Scale3D=(X=1,Y=1,Z=1))),(Item=(Type=Bone,Name=\"SkirtB_r_03_dyn\"),Transform=(Rotation=(X=0,Y=0.0138,Z=0,W=0.9999),Translation=(X= ...(truncated)"
    WeightMinimum=0
    Weight=@n22.Result
   )
   n23 = RigVMVariableNode() {variable=is DBNO}
   n24 = &RigVMFunction_AlphaInterp(
    InterpSpeedDecreasing=4
    InterpSpeedIncreasing=10
    Scale=1
    bClampResult=False
    Value=&RigVMFunction_MathDoubleMul(
     A=RigVMDispatchNode(False=0, True=1, Condition=@n23.Value) {template="DISPATCH_RigVMDispatch_If(in Condition,in True,in False,out Result)"}.Result
     B=RigVMVariableNode() {variable=is_Moving_Back}.Value
    ).Result
   )
   > &RigUnit_ModifyTransforms(
    ItemToModify="((Item=(Type=Bone,Name=\"SkirtC_r_02_dyn\"),Transform=(Rotation=(X=0,Y=0,Z=-0.3984,W=0.9172),Translation=(X=2.4425,Y=1.2958,Z=0),Scale3D=(X=1,Y=1,Z=1))),(Item=(Type=Bone,Name=\"SkirtB_02_dyn\"),Transform=(Rotation=(X=-0.01,Y=-0.0917,Z=-0.0007,W=0.9957),Transla ...(truncated)"
    WeightMinimum=0
    Weight=@n24.Result
   )
   n25 = &RigVMFunction_AlphaInterp(
    InterpSpeedDecreasing=8
    InterpSpeedIncreasing=10
    Scale=1
    bClampResult=False
    Value=&RigVMFunction_MathDoubleMul(A=RigVMVariableNode() {variable=is_Moving_Back}.Value, B=RigVMVariableNode() {variable=is_Jumping_CR}.Value).Result
   )
   &RigUnit_ModifyTransforms(
    ItemToModify=((Item=(Type=Bone,Name="SkirtC_r_03_dyn"),Transform=(Rotation=(X=0.2278,Y=0.0719,Z=-0.0105,W=-0.971),Translation=(X=-3.7596,Y=6.6021,Z=-2.4293),Scale3D=(X=1,Y=1,Z=1))))
    WeightMinimum=0
    Weight=@n25.Result
    ExecutePin=@RigVMFunction_Sequence.G
   )
   RigVMFunctionReturnNode(ExecuteContext=@RigVMFunction_Sequence.I)
   &RigUnit_ModifyTransforms(
    ItemToModify=((Item=(Type=Bone,Name="foot_l_dyn_deform"),Transform=(Rotation=(X=0,Y=0.342,Z=0,W=0.9397),Translation=(X=6,Y=0,Z=-1.5),Scale3D=(X=1,Y=1,Z=1))))
    WeightMinimum=0
    Weight=RigVMVariableNode() {variable=SPR L Foot}.Value
    ExecutePin=@RigVMFunction_Sequence.A
   )
   > &RigUnit_ModifyTransforms(
    ItemToModify=((Item=(Type=Bone,Name="foot_r_dyn_deform"),Transform=(Rotation=(X=0,Y=0.1736,Z=0,W=0.9848),Translation=(X=6,Y=0,Z=-1),Scale3D=(X=1,Y=1,Z=1))))
    WeightMinimum=0
    Weight=RigVMVariableNode() {variable=SPR R Foot}.Value
   )
   > DeltaRotation v4_1 = RigVMFunctionReferenceNode(
    Rotation Order=ZYX
    child=(Type=Bone,Name="thigh_r")
    debugRotation=false
    outputMinMax=(X=0,Y=1)
    parent=(Type=Null,Name="thigh_r_null")
    premult=1
    rotationMinMax=(X=0,Y=-13)
    triggerAxis=Y
    twistAxis=(X=1,Y=0,Z=0)
   )
   > &RigUnit_ModifyTransforms(
    ItemToModify="((Item=(Type=Bone,Name=\"None\"),Transform=(Rotation=(X=0,Y=0,Z=0,W=1),Translation=(X=0,Y=0,Z=0),Scale3D=(X=1,Y=1,Z=1))),(Item=(Type=Bone,Name=\"SkirtC_r_02_dyn\"),Transform=(Rotation=(X=0.0082,Y=0.1637,Z=-0.0495,W=-0.9852),Translation=(X=0,Y=0,Z=0),Scale3D=(X ...(truncated)"
    WeightMinimum=0
    Weight=&RigVMFunction_MathDoubleMul(A=@"DeltaRotation v4_1".outputWeight, B=RigVMVariableNode() {variable=Parachute_Open}.Value).Result
   )
   > &RigUnit_ModifyTransforms(
    ItemToModify="((Item=(Type=Bone,Name=\"None\"),Transform=(Rotation=(X=0,Y=0,Z=0,W=1),Translation=(X=0,Y=0,Z=0),Scale3D=(X=1,Y=1,Z=1))),(Item=(Type=Bone,Name=\"SkirtF_01_dyn\"),Transform=(Rotation=(X=0,Y=-0.373,Z=0,W=-0.9278),Translation=(X=0,Y=0,Z=0),Scale3D=(X=1,Y=1,Z=1))) ...(truncated)"
    WeightMinimum=0
    Weight=RigVMVariableNode() {variable=Is_Griddles_Emote}.Value
   )
   AlphaInterp = &RigVMFunction_AlphaInterp(
    InterpSpeedDecreasing=7
    InterpSpeedIncreasing=1
    Scale=1
    bClampResult=False
    bMapRange=false
    Value=RigVMVariableNode() {variable=Is_BreakDance_Emote}.Value
   )
   > &RigUnit_ModifyTransforms(
    ItemToModify="((Item=(Type=Bone,Name=\"None\"),Transform=(Rotation=(X=0,Y=0,Z=0,W=1),Translation=(X=0,Y=0,Z=0),Scale3D=(X=1,Y=1,Z=1))),(Item=(Type=Bone,Name=\"SkirtB_01_dyn\"),Transform=(Rotation=(X=0.0008,Y=0.1536,Z=-0.2856,W=0.946),Translation=(X=-0.6623,Y=-6.1496,Z=0.085 ...(truncated)"
    WeightMinimum=0
    Weight=@AlphaInterp.Result
   )
   > DeltaRotation v4 = RigVMFunctionReferenceNode(
    Rotation Order=ZYX
    child=(Type=Bone,Name="calf_r")
    debugRotation=false
    outputMinMax=(X=0,Y=1)
    parent=(Type=Null,Name="calf_r_null")
    premult=1
    rotationMinMax=(X=110,Y=116)
    triggerAxis=Z
    twistAxis=(X=1,Y=0,Z=0)
   )
   > DeltaRotation v4_18 = RigVMFunctionReferenceNode(
    Rotation Order=ZYX
    child=(Type=Bone,Name="foot_r")
    debugRotation=false
    onlyTwist=false
    outputMinMax=(X=0,Y=1)
    parent=(Type=Null,Name="foot_r_null")
    premult=1
    rotationMinMax=(X=-40,Y=-50)
    triggerAxis=Z
    twistAxis=(X=1,Y=0,Z=0)
   )
   Multiply_12 = RigVMAggregateNode(A=@"DeltaRotation v4".outputWeight, B=@"DeltaRotation v4_18".outputWeight, C=RigVMVariableNode() {variable=Is_Crouching}.Value)
   n26 = &RigVMFunction_AlphaInterp(InterpSpeedDecreasing=1, InterpSpeedIncreasing=5, Scale=1, bClampResult=False, Value=@Multiply_12.Result)
   > &RigUnit_ModifyTransforms(
    ItemToModify="((Item=(Type=Bone,Name=\"SkirtC_r_01_dyn\"),Transform=(Rotation=(X=0.0086,Y=0.0911,Z=-0.0938,W=-0.9914),Translation=(X=6.1712,Y=0.7283,Z=1.1434),Scale3D=(X=1,Y=1,Z=1))),(Item=(Type=Bone,Name=\"SkirtB_r_02_dyn\"),Transform=(Rotation=(X=0,Y=0,Z=0,W=1),Translatio ...(truncated)"
    WeightMinimum=0
    Weight=@n26.Result
   )
   &RigUnit_ModifyTransforms(
    ItemToModify="((Item=(Type=Bone,Name=\"upperarm_l_root_dyn\"),Transform=(Rotation=(X=-0.0038,Y=-0.0435,Z=-0.0871,W=0.9952),Translation=(X=0.0001,Y=-1.5008,Z=-3),Scale3D=(X=1,Y=1,Z=1))),(Item=(Type=Bone,Name=\"upperarm_l_low_dyn\"),Transform=(Rotation=(X=0,Y=0,Z=0,W=1),Trans ...(truncated)"
    WeightMinimum=0
    Weight=RigVMDispatchNode(BoneName=root, Default=1, Name=deformRig_l_upperarm_up_alpha) {template="DISPATCH_RigDispatch_GetAnimAttribute(in Name,in BoneName,in Default,out Value,out Found)"}.Value
    ExecutePin=@RigVMFunction_Sequence.C
   )
   > &RigUnit_ModifyTransforms(
    ItemToModify="((Item=(Type=Bone,Name=\"upperarm_r_root_dyn\"),Transform=(Rotation=(X=0.2105,Y=0.0683,Z=-0.0169,W=0.9751),Translation=(X=-0.0001,Y=0,Z=0),Scale3D=(X=1,Y=0.8,Z=1))),(Item=(Type=Bone,Name=\"upperarm_r_low_dyn\"),Transform=(Rotation=(X=0.0851,Y=0.0189,Z=0.2156,W ...(truncated)"
    WeightMinimum=0
    Weight=RigVMDispatchNode(BoneName=root, Default=1, Name=deformRig_r_upperarm_fr_alpha) {template="DISPATCH_RigDispatch_GetAnimAttribute(in Name,in BoneName,in Default,out Value,out Found)"}.Value
   )
   > &RigUnit_ModifyTransforms(
    ItemToModify=((Item=(Type=Bone,Name="None"),Transform=(Rotation=(X=0,Y=0,Z=0,W=1),Translation=(X=0,Y=0,Z=0),Scale3D=(X=1,Y=1,Z=1))),(Item=(Type=Bone,Name="upperarm_r_root_dyn"),Transform=(Rotation=(X=0,Y=0,Z=0,W=1),Translation=(X=0,Y=0,Z=-3),Scale3D=(X=1,Y=1,Z=1))))
    WeightMinimum=0
    Weight=RigVMDispatchNode(BoneName=root, Default=1, Name=deformRig_r_upperarm_up_alpha) {template="DISPATCH_RigDispatch_GetAnimAttribute(in Name,in BoneName,in Default,out Value,out Found)"}.Value
   )
   > DeltaRotation v4_11 = RigVMFunctionReferenceNode(
    Rotation Order=ZYX
    child=(Type=Bone,Name="spine_05")
    debugRotation=false
    outputMinMax=(X=0,Y=1)
    parent=(Type=Null,Name="spine_05_null")
    premult=1
    rotationMinMax=(X=0,Y=8)
    triggerAxis=Z
    twistAxis=(X=1,Y=0,Z=0)
   )
   n27 = &RigVMFunction_AlphaInterp(
    InterpSpeedDecreasing=2
    InterpSpeedIncreasing=2.5
    Scale=1
    bClampResult=False
    Value=&RigVMFunction_MathDoubleMul(A=@"DeltaRotation v4_11".outputWeight, B=RigVMVariableNode() {variable=Is_pickaxe_Swing}.Value).Result
   )
   > &RigUnit_ModifyTransforms(
    ItemToModify="((Item=(Type=Bone,Name=\"None\"),Transform=(Rotation=(X=0,Y=0,Z=0,W=1),Translation=(X=0,Y=0,Z=0),Scale3D=(X=1,Y=1,Z=1))),(Item=(Type=Bone,Name=\"SkirtB_r_01_dyn\"),Transform=(Rotation=(X=0.1244,Y=0.0061,Z=0.1623,W=-0.9789),Translation=(X=0,Y=0,Z=0),Scale3D=(X= ...(truncated)"
    WeightMinimum=0
    Weight=@n27.Result
   )
   > DeltaRotation v4_3 = RigVMFunctionReferenceNode(
    Rotation Order=ZYX
    child=(Type=Bone,Name="thigh_r")
    debugRotation=false
    outputMinMax=(X=0,Y=1)
    parent=(Type=Null,Name="thigh_r_null")
    premult=1
    rotationMinMax=(X=-25,Y=-40)
    triggerAxis=Z
    twistAxis=(X=1,Y=0,Z=0)
   )
   n28 = &RigVMFunction_AlphaInterp(
    InterpSpeedDecreasing=20
    InterpSpeedIncreasing=5
    Scale=1
    bClampResult=False
    Value=&RigVMFunction_MathDoubleMul(A=@"DeltaRotation v4_3".outputWeight, B=RigVMVariableNode() {variable=Parachute_Open}.Value).Result
   )
   > &RigUnit_ModifyTransforms(
    ItemToModify="((Item=(Type=Bone,Name=\"SkirtA_r_01_dyn\"),Transform=(Rotation=(X=-0.0064,Y=0.2686,Z=-0.0842,W=-0.9595),Translation=(X=0,Y=0,Z=0),Scale3D=(X=1,Y=1.5,Z=1))),(Item=(Type=Bone,Name=\"SkirtF_01_dyn\"),Transform=(Rotation=(X=-0.2649,Y=0.1498,Z=0.2299,W=-0.9244),Tr ...(truncated)"
    WeightMinimum=0
    Weight=@n28.Result
   )
   > DeltaRotation v4_2 = RigVMFunctionReferenceNode(
    Rotation Order=ZYX
    child=(Type=Bone,Name="thigh_l")
    debugRotation=false
    onlyTwist=false
    outputMinMax=(X=0,Y=1)
    parent=(Type=Null,Name="thigh_l_null")
    premult=1
    rotationMinMax=(X=-20,Y=-40)
    triggerAxis=Z
    twistAxis=(X=1,Y=0,Z=0)
   )
   n29 = &RigVMFunction_AlphaInterp(
    InterpSpeedDecreasing=20
    InterpSpeedIncreasing=5
    Scale=1
    bClampResult=False
    Value=&RigVMFunction_MathDoubleMul(A=@"DeltaRotation v4_2".outputWeight, B=RigVMVariableNode() {variable=Parachute_Open}.Value).Result
   )
   > &RigUnit_ModifyTransforms(
    ItemToModify="((Item=(Type=Bone,Name=\"SkirtA_l_01_dyn\"),Transform=(Rotation=(X=0.0095,Y=0.0143,Z=-0.0497,W=-0.9986),Translation=(X=0,Y=0,Z=0),Scale3D=(X=1,Y=1.5,Z=1))),(Item=(Type=Bone,Name=\"SkirtF_01_dyn\"),Transform=(Rotation=(X=0.006,Y=0.3525,Z=-0.0175,W=-0.9356),Tran ...(truncated)"
    WeightMinimum=0
    Weight=@n29.Result
   )
   DeltaRotation v4_4 = RigVMFunctionReferenceNode(
    Rotation Order=ZYX
    child=(Type=Bone,Name="lowerarm_r")
    debugRotation=false
    outputMinMax=(X=0,Y=1)
    parent=(Type=Null,Name="lowerarm_r_null")
    premult=1
    rotationMinMax=(X=0,Y=-0.1)
    triggerAxis=Z
    twistAxis=(X=1,Y=0,Z=0)
    ExecuteContext=@RigVMFunction_Sequence.H
   )
   > &RigUnit_ModifyTransforms(
    ItemToModify=((Item=(Type=Bone,Name="lowerarm_r"),Transform=(Rotation=(X=0,Y=0,Z=0.0872,W=0.9962),Translation=(X=0,Y=0,Z=0),Scale3D=(X=1,Y=1,Z=1))))
    WeightMinimum=0
    Weight=&RigVMFunction_MathDoubleMul(A=@"DeltaRotation v4_4".outputWeight, B=RigVMVariableNode() {variable=Facial_Testing}.Value).Result
   )
   > DeltaRotation v4_5 = RigVMFunctionReferenceNode(
    Rotation Order=ZYX
    child=(Type=Bone,Name="lowerarm_l")
    debugRotation=false
    outputMinMax=(X=0,Y=1)
    parent=(Type=Null,Name="lowerarm_l_null")
    premult=1
    rotationMinMax=(X=0,Y=-0.1)
    triggerAxis=Z
    twistAxis=(X=1,Y=0,Z=0)
   )
   > &RigUnit_ModifyTransforms(
    ItemToModify=((Item=(Type=Bone,Name="lowerarm_l"),Transform=(Rotation=(X=0,Y=0,Z=0.0872,W=0.9962),Translation=(X=0,Y=0,Z=0),Scale3D=(X=1,Y=1,Z=1))))
    WeightMinimum=0
    Weight=&RigVMFunction_MathDoubleMul(A=@"DeltaRotation v4_5".outputWeight, B=RigVMVariableNode() {variable=Facial_Testing}.Value).Result
   )
   > DeltaRotation v4_6 = RigVMFunctionReferenceNode(
    Rotation Order=ZYX
    child=(Type=Bone,Name="upperarm_r")
    debugRotation=false
    outputMinMax=(X=0,Y=3)
    parent=(Type=Null,Name="upperarm_r_null")
    premult=1
    rotationMinMax=(X=-5,Y=-2)
    triggerAxis=Z
    twistAxis=(X=1,Y=0,Z=0)
   )
   n30 = &RigVMFunction_MathDoubleMul(A=@"DeltaRotation v4_6".outputWeight, B=RigVMVariableNode() {variable=Facial_Testing}.Value)
   > &RigUnit_ModifyTransforms(
    ItemToModify=((Item=(Type=Bone,Name="upperarm_r_low_dyn"),Transform=(Rotation=(X=0,Y=-0.0872,Z=0,W=0.9962),Translation=(X=0,Y=0,Z=0),Scale3D=(X=1,Y=1,Z=1))))
    WeightMinimum=0
    Weight=@n30.Result
   )
   > &RigUnit_ModifyTransforms(
    ItemToModify=((Item=(Type=Bone,Name="upperarm_r"),Transform=(Rotation=(X=0,Y=-0.0175,Z=0,W=0.9998),Translation=(X=0,Y=0,Z=0),Scale3D=(X=1,Y=1,Z=1))))
    WeightMinimum=0
    Weight=@n30.Result
   )
   > DeltaRotation v4_7 = RigVMFunctionReferenceNode(
    Rotation Order=ZYX
    child=(Type=Bone,Name="upperarm_l")
    debugRotation=false
    outputMinMax=(X=0,Y=3)
    parent=(Type=Null,Name="upperarm_l_null")
    premult=1
    rotationMinMax=(X=-5,Y=-2)
    triggerAxis=Z
    twistAxis=(X=1,Y=0,Z=0)
   )
   n31 = &RigVMFunction_MathDoubleMul(A=@"DeltaRotation v4_7".outputWeight, B=RigVMVariableNode() {variable=Facial_Testing}.Value)
   > &RigUnit_ModifyTransforms(
    ItemToModify=((Item=(Type=Bone,Name="upperarm_l_low_dyn"),Transform=(Rotation=(X=0,Y=-0.0872,Z=0,W=0.9962),Translation=(X=0,Y=0,Z=0),Scale3D=(X=1,Y=1,Z=1))))
    WeightMinimum=0
    Weight=@n31.Result
   )
   > &RigUnit_ModifyTransforms(
    ItemToModify=((Item=(Type=Bone,Name="upperarm_l"),Transform=(Rotation=(X=0,Y=-0.0523,Z=0,W=0.9986),Translation=(X=0,Y=0,Z=0),Scale3D=(X=1,Y=1,Z=1))))
    WeightMinimum=0
    Weight=@n31.Result
   )
   &RigUnit_ModifyTransforms(
    ExecutePin=()
    ItemToModify=((Item=(Type=Bone,Name="SkirtC_r_01_dyn"),Transform=(Rotation=(X=0,Y=-0.1215,Z=0,W=0.9926),Translation=(X=6.7594,Y=0,Z=-5.4272),Scale3D=(X=1,Y=1,Z=1))))
    Weight=1
    WeightMinimum=0
   )
  ]}
 New Function_ContainedGraph {nodes [
   RigVMFunctionEntryNode()
   > RigVMFunctionReturnNode()
  ]}
 RigVMFunctionLibrary {nodes [
   Apron_Fix = RigVMCollapseNode(ExecuteContext=())
   Controlers = RigVMCollapseNode(Debug Controls=False, ExecuteContext=())
   DeltaRotation v4 = RigVMCollapseNode(
    ExecuteContext=()
    Rotation Order=ZYX
    child=(Type=Bone,Name="None")
    outputMinMax=(X=0,Y=1)
    parent=(Type=Bone,Name="None")
    premult=1
    rotationMinMax=(X=0,Y=0)
    twistAxis=(X=1,Y=0,Z=0)
   )
   New Function = RigVMCollapseNode(ExecuteContext=())
   Side Sleeves Fixes = RigVMCollapseNode(ExecuteContext=())
   Variables = RigVMCollapseNode(ExecuteContext=())
  ]}
 RigVMFunction_Sequence_1_ContainedGraph {nodes [
   n1 = RigVMFunctionEntryNode()
   RigVMFunction_Sequence = &RigVMFunction_Sequence(ExecuteContext=@n1.ExecuteContext)
   n2 = &RigVMFunction_Sequence(ExecuteContext=@RigVMFunction_Sequence.B)
   n3 = &RigVMFunction_Sequence(ExecuteContext=@n2.B)
   n4 = &RigVMFunction_Sequence(ExecuteContext=@n3.B)
   n5 = &RigVMFunction_Sequence(ExecuteContext=@n4.B)
   RigVMFunctionReturnNode(A=@RigVMFunction_Sequence.A, B=@n2.A, C=@n3.A, G=@n4.A, H=@n5.A, I=@n5.B)
  ]}
 RigVMModel {nodes [
   RigUnit_BeginExecution = &RigUnit_BeginExecution()
   Controlers = RigVMFunctionReferenceNode(Debug Controls=false, ExecuteContext=@RigUnit_BeginExecution.ExecutePin)
   > FN_DeformRig_FNC = RigVMFunctionReferenceNode(
    DeformRig_ Curve="(deformRig_l_upperarm_up_alpha_72_#1=true,deformRig_r_upperarm_up_alpha_73_#2=true,deformRig_l_upperarm_fr_alpha_74_#3=False,deformRig_r_upperarm_fr_alpha_75_CB51B80C ...(truncated)"
    DeformRig_Bool="(upperarm_1_#4=True,elbow_3_#5=True,wrist_7_#6=True,groin_8_#7=True,glute_17_#8=True,knee_9_40C408324CA261 ...(truncated)"
    DeformRig_Debug=(upperarm_debug_18_#4=False,wrist_debug_20_#6=False,lat_debug_26_#9=False,groinGlute_debug_27_#7=false)
    DeformRig_Value="(upperarm_up_18_#10=(Rotation=(X=0.0,Y=-0.0,Z=0.0,W=1.0),Translation=(X=0.0,Y=0.0,Z=8.0),Scale3D=(X=0.70,Y=1.0,Z=1.0)),upperarm_fwd_fr_19_#11=(Rotation=(X=0.0,Y=-0.0,Z=0.0,W=1.0),Translation=(X=0.0, ...(truncated)"
   )
   > Variables = RigVMFunctionReferenceNode()
   > Apron_Fix = RigVMFunctionReferenceNode()
   > Side Sleeves Fixes = RigVMFunctionReferenceNode()
  ]}
]
hierarchy [
 root {loc=<0 0 0>, rot=<0 0 0 1>}
 pelvis {parent=root, loc=<0 -0.7071 93.1191>, rot=<0.0127 0.707 -0.0127 -0.707>}
 spine_01 {parent=pelvis, loc=<6.7494 -0.6388 0>, rot=<0 0 0.034 -0.9994>}
 spine_02 {parent=spine_01, loc=<7.2518 -0.035 0>, rot=<0 0 0.0238 -0.9997>}
 spine_03 {parent=spine_02, loc=<6.0104 0 0>, rot=<0 0 0.0149 -0.9999>}
 spine_04 {parent=spine_03, loc=<8.1024 0 0>, rot=<0 0 -0.0132 -0.9999>}
 spine_05 {parent=spine_04, loc=<10.0146 0 0>, rot=<0 0 -0.0015 -1>}
 neck_01 {parent=spine_05, loc=<11.3138 0.1529 0>, rot=<0 0 -0.2003 -0.9797>}
 neck_02 {parent=neck_01, loc=<4.5627 0 0>, rot=<0 0 0.0504 -0.9987>}
 head {parent=neck_02, loc=<5.3914 0 0>, rot=<0 0 0.115 -0.9934>}
 attach_cape {parent=spine_05, loc=<9.0163 -11.7133 0.0015>, rot=<0.0616 -0.7044 -0.0616 -0.7044>}
 cape {parent=attach_cape, type=Null, loc=<0 -1.9225 4.093>, rot=<0 0 0 1>}
 attach_backpack {parent=spine_05, loc=<-0.1219 -13.4715 0.0011>, rot=<-0.0092 -0.707 0.0093 -0.707>}
 Backpack_BR {parent=attach_backpack, type=Null, loc=<0 -3.1945 -0.2243>, rot=<-0.0664 0 0 0.9978>}
 clavicle_l {parent=spine_05, loc=<4.3575 -0.1466 -2.3223>, rot=<0.0465 -0.7289 -0.009 -0.6829>}
 upperarm_l_null {parent=clavicle_l, type=Null, loc=<15.5111 0 0>, rot=<0.0226 -0.3234 -0.0088 -0.9459>}
 upperarm_l {parent=clavicle_l, loc=<15.5111 0 0>, rot=<0.0226 -0.3234 -0.0088 -0.9459>}
 lowerarm_l {parent=upperarm_l, loc=<26.6514 0 0>, rot=<0.0391 -0.0099 -0.2453 -0.9686>}
 hand_l {parent=lowerarm_l, loc=<22.3491 0 0>, rot=<-0.7081 -0.0507 0.0975 -0.6975>}
 thumb_01_l {parent=hand_l, loc=<1.3118 -0.3191 -2.459>, rot=<0.6546 -0.4042 0.0664 -0.6354>}
 thumb_02_l {parent=thumb_01_l, loc=<5.9336 -0.0448 0.0401>, rot=<-0.0037 0.0213 0.298 -0.9543>}
 thumb_03_l {parent=thumb_02_l, loc=<2.2223 0.3682 -0.0907>, rot=<0.0118 -0.013 0.0066 -0.9998>}
 index_metacarpal_l {parent=hand_l, loc=<1.6535 0.0183 -2.96>, rot=<0.0158 -0.0124 -0.0173 -0.9996>}
 index_01_l {parent=index_metacarpal_l, loc=<8.3008 0.0357 -0.0524>, rot=<-0.0044 0.0485 0.0924 -0.9945>}
 index_02_l {parent=index_01_l, loc=<3.3587 -0.3004 0.0452>, rot=<-0.002 0.0027 0.3225 -0.9466>}
 index_03_l {parent=index_02_l, loc=<2.5578 -0.123 -0.0252>, rot=<-0.002 0.0067 0.2558 -0.9667>}
 middle_metacarpal_l {parent=hand_l, loc=<2.1856 -0.1725 -0.5984>, rot=<0.006 0.0168 -0.0329 -0.9993>}
 middle_01_l {parent=middle_metacarpal_l, loc=<7.3275 -0.2447 0.0129>, rot=<-0.0208 0.0042 0.0858 -0.9961>}
 middle_02_l {parent=middle_01_l, loc=<4.0123 0.1401 0.1175>, rot=<-0.0218 0.0258 0.3512 -0.9357>}
 middle_03_l {parent=middle_02_l, loc=<2.6077 0.0534 0.1373>, rot=<0.0082 0.0188 0.1708 -0.9851>}
 ring_metacarpal_l {parent=hand_l, loc=<2.2074 -0.3625 1.3638>, rot=<-0.0284 0.0722 -0.0458 -0.9959>}
 ring_01_l {parent=ring_metacarpal_l, loc=<7.0295 -0.6197 -0.1446>, rot=<-0.0489 -0.0252 0.1892 -0.9804>}
 ring_02_l {parent=ring_01_l, loc=<3.6174 0.4647 -0.1027>, rot=<0 0 0.2806 -0.9598>}
 ring_03_l {parent=ring_02_l, loc=<2.4152 0.114 0.0199>, rot=<0 0 0.1189 -0.9929>}
 pinky_metacarpal_l {parent=hand_l, loc=<1.7616 -0.9753 3.3814>, rot=<-0.0397 0.0814 -0.069 -0.9935>}
 pinky_01_l {parent=pinky_metacarpal_l, loc=<7.1709 -0.4831 -0.1094>, rot=<-0.0654 -0.0035 0.2242 -0.9723>}
 pinky_02_l {parent=pinky_01_l, loc=<2.8274 0.0893 -0.1043>, rot=<0.0037 0.0011 0.2854 -0.9584>}
 pinky_03_l {parent=pinky_02_l, loc=<2.2046 0.0307 -0.0078>, rot=<-0.0002 0.0098 0.1445 -0.9895>}
 deform_wrist_dn_base_l {parent=hand_l, loc=<0 -2.2342 0>, rot=<0 0 0 -1>}
 deform_wrist_dn_l {parent=deform_wrist_dn_base_l, loc=<0 0 0>, rot=<0 0 0 1>}
 deform_wrist_up_base_l {parent=hand_l, loc=<0 2.2682 0>, rot=<0 0 0 -1>}
 deform_wrist_up_l {parent=deform_wrist_up_base_l, loc=<0 0 0>, rot=<0 0 0 1>}
 lowerarm_twist_02_l {parent=lowerarm_l, loc=<9.6199 0 0>, rot=<0 0 0 -1>}
 lowerarm_twist_01_l {parent=lowerarm_l, loc=<17.8549 0 0>, rot=<0 0 0 -1>}
 deform_elbow_in_base_l {parent=lowerarm_l, loc=<0.2112 5.259 0.8274>, rot=<0 0 0 -1>}
 deform_elbow_in_l {parent=deform_elbow_in_base_l, loc=<0 0 0>, rot=<0 0 0 1>}
 deform_elbow_out_base_l {parent=lowerarm_l, loc=<0 -4 0>, rot=<0 0 0 -1>}
 deform_elbow_out_l {parent=deform_elbow_out_base_l, loc=<0 0 0>, rot=<0 0 0 1>}
 upperarm_twist_01_l {parent=upperarm_l, loc=<0.0041 0 0>, rot=<0 0 0 -1>}
 upperarm_twist_02_l {parent=upperarm_l, loc=<13.7941 0 0>, rot=<0 0 0 -1>}
 upperarm_l_root_dyn {parent=upperarm_twist_02_l, loc=<-4.573 0 0>, rot=<0 0 0 1>}
 upperarm_l_low_base {parent=upperarm_l_root_dyn, loc=<-2.2718 -0.594 4.3184>, rot=<-0.0053 -0.0606 -0.0002 -0.9981>}
 upperarm_l_low_dyn {parent=upperarm_l_low_base, loc=<0 0 0>, rot=<0 0 0 1>}
 lowerarm_l_null {parent=upperarm_l, type=Null, loc=<26.6514 0 0>, rot=<0.0391 -0.0099 -0.2453 -0.9686>}
 clavicle_r {parent=spine_05, loc=<4.3575 -0.1466 2.3223>, rot=<0.7289 0.0465 -0.6829 0.009>}
 upperarm_r_null {parent=clavicle_r, type=Null, loc=<-15.5111 0 0>, rot=<0.0226 -0.3234 -0.0088 -0.9459>}
 upperarm_r {parent=clavicle_r, loc=<-15.5111 0 0>, rot=<0.0226 -0.3234 -0.0088 -0.9459>}
 lowerarm_r_null {parent=upperarm_r, type=Null, loc=<-26.6514 0 0>, rot=<0.0391 -0.0099 -0.2453 -0.9686>}
 lowerarm_r {parent=upperarm_r, loc=<-26.6514 0 0>, rot=<0.0391 -0.0099 -0.2453 -0.9686>}
 hand_r {parent=lowerarm_r, loc=<-22.3491 0 0>, rot=<-0.7081 -0.0507 0.0975 -0.6975>}
 thumb_01_r {parent=hand_r, loc=<-1.3118 0.3191 2.459>, rot=<0.6546 -0.4042 0.0664 -0.6354>}
 thumb_02_r {parent=thumb_01_r, loc=<-5.9336 0.0448 -0.0401>, rot=<-0.0037 0.0213 0.298 -0.9543>}
 thumb_03_r {parent=thumb_02_r, loc=<-2.2223 -0.3682 0.0907>, rot=<0.0118 -0.013 0.0066 -0.9998>}
 index_metacarpal_r {parent=hand_r, loc=<-1.6535 -0.0183 2.96>, rot=<0.0158 -0.0124 -0.0173 -0.9996>}
 index_01_r {parent=index_metacarpal_r, loc=<-8.3008 -0.0357 0.0524>, rot=<-0.0044 0.0485 0.0924 -0.9945>}
 index_02_r {parent=index_01_r, loc=<-3.3587 0.3004 -0.0452>, rot=<-0.002 0.0027 0.3225 -0.9466>}
 index_03_r {parent=index_02_r, loc=<-2.5578 0.123 0.0252>, rot=<-0.002 0.0067 0.2558 -0.9667>}
 middle_metacarpal_r {parent=hand_r, loc=<-2.1856 0.1725 0.5984>, rot=<0.006 0.0168 -0.0329 -0.9993>}
 middle_01_r {parent=middle_metacarpal_r, loc=<-7.3275 0.2447 -0.0129>, rot=<-0.0208 0.0042 0.0858 -0.9961>}
 middle_02_r {parent=middle_01_r, loc=<-4.0123 -0.1401 -0.1175>, rot=<-0.0218 0.0258 0.3512 -0.9357>}
 middle_03_r {parent=middle_02_r, loc=<-2.6077 -0.0534 -0.1373>, rot=<0.0082 0.0188 0.1708 -0.9851>}
 ring_metacarpal_r {parent=hand_r, loc=<-2.2074 0.3625 -1.3638>, rot=<-0.0284 0.0722 -0.0458 -0.9959>}
 ring_01_r {parent=ring_metacarpal_r, loc=<-7.0295 0.6197 0.1446>, rot=<-0.0489 -0.0252 0.1892 -0.9804>}
 ring_02_r {parent=ring_01_r, loc=<-3.6174 -0.4647 0.1027>, rot=<0 0 0.2806 -0.9598>}
 ring_03_r {parent=ring_02_r, loc=<-2.4152 -0.114 -0.0199>, rot=<0 0 0.1189 -0.9929>}
 pinky_metacarpal_r {parent=hand_r, loc=<-1.7616 0.9753 -3.3814>, rot=<-0.0397 0.0814 -0.069 -0.9935>}
 pinky_01_r {parent=pinky_metacarpal_r, loc=<-7.1709 0.4831 0.1094>, rot=<-0.0654 -0.0035 0.2242 -0.9723>}
 pinky_02_r {parent=pinky_01_r, loc=<-2.8274 -0.0893 0.1043>, rot=<0.0037 0.0011 0.2854 -0.9584>}
 pinky_03_r {parent=pinky_02_r, loc=<-2.2046 -0.0307 0.0078>, rot=<-0.0002 0.0098 0.1445 -0.9895>}
 deform_wrist_dn_base_r {parent=hand_r, loc=<-0.0004 2.2345 -0.0001>, rot=<0 0 0 -1>}
 deform_wrist_dn_r {parent=deform_wrist_dn_base_r, loc=<0 0 0>, rot=<0 0 0 1>}
 deform_wrist_up_base_r {parent=hand_r, loc=<0.0003 -2.2684 0.0001>, rot=<0 0 0 -1>}
 deform_wrist_up_r {parent=deform_wrist_up_base_r, loc=<0 0 0>, rot=<0 0 0 1>}
 lowerarm_twist_02_r {parent=lowerarm_r, loc=<-9.6199 0 0>, rot=<0 0 0 -1>}
 lowerarm_twist_01_r {parent=lowerarm_r, loc=<-17.8549 0 0>, rot=<0 0 0 -1>}
 deform_elbow_in_base_r {parent=lowerarm_r, loc=<-0.2111 -5.2591 -0.8275>, rot=<0 0 0 -1>}
 deform_elbow_in_r {parent=deform_elbow_in_base_r, loc=<-0.0001 0 0.0002>, rot=<0 0 0 1>}
 deform_elbow_out_base_r {parent=lowerarm_r, loc=<0 4 -0.0001>, rot=<0 0 0 -1>}
 deform_elbow_out_r {parent=deform_elbow_out_base_r, loc=<0 0 0>, rot=<0 0 0 1>}
 upperarm_twist_01_r {parent=upperarm_r, loc=<-0.0041 0 0>, rot=<0 0 0 -1>}
 upperarm_twist_02_r {parent=upperarm_r, loc=<-13.7941 0 0>, rot=<0 0 0 -1>}
 upperarm_r_root_dyn {parent=upperarm_twist_02_r, loc=<4.5733 0 -0.0004>, rot=<0 -1 0 0>}
 upperarm_r_low_base {parent=upperarm_r_root_dyn, loc=<-2.2713 0.594 4.3179>, rot=<0.0053 -0.0606 0.0002 -0.9981>}
 upperarm_r_low_dyn {parent=upperarm_r_low_base, loc=<0 0 0>, rot=<0 0 0 1>}
 bow_l {parent=spine_05, loc=<7.9813 11.7632 -6.3296>, rot=<-0.1072 -0.6802 0.7021 0.1816>}
 bow_l_dyn_01 {parent=bow_l, loc=<0 0 0>, rot=<0 0 0 1>}
 bow_l_dyn_02 {parent=bow_l_dyn_01, loc=<3.8609 0 0>, rot=<0 0 0 1>}
 bow_r {parent=spine_05, loc=<7.981 11.7632 6.3296>, rot=<0.6802 -0.1072 0.1816 -0.7021>}
 bow_r_dyn_01 {parent=bow_r, loc=<0 0 0>, rot=<0 0 0 1>}
 bow_r_dyn_02 {parent=bow_r_dyn_01, loc=<-3.8605 0.0001 0.0002>, rot=<0 0 0 -1>}
 bow_base {parent=spine_05, loc=<8.5686 13.3991 0>, rot=<-0.5625 -0.4285 -0.4285 -0.5625>}
 bow_c {parent=bow_base, loc=<2.2274 0 -0.0244>, rot=<0 0 0 1>}
 bow_l_base {parent=bow_c, loc=<-0.2368 -3.0184 -0.1084>, rot=<0.0266 -0.0189 0.7099 -0.7035>}
 bow_l_dyn {parent=bow_l_base, loc=<0 0 0>, rot=<0 0 0 1>}
 bow_r_base {parent=bow_c, loc=<-0.2367 3.0184 -0.1084>, rot=<0.7099 -0.7035 -0.0266 0.0189>}
 bow_r_dyn {parent=bow_r_base, loc=<0 0 0>, rot=<0 0 0 1>}
 thigh_l {parent=pelvis, loc=<-2.5693 -0.0109 -7.7501>, rot=<0.0012 0.0572 0.0155 -0.9982>}
 calf_l {parent=thigh_l, loc=<-39.0484 0 0>, rot=<0.0006 -0.0161 -0.0391 -0.9991>}
 foot_l {parent=calf_l, loc=<-42.561 0.1077 0.4292>, rot=<0.0418 -0.0455 0.0343 -0.9975>}
 ball_l {parent=foot_l, loc=<-6.1599 13.8264 0.6806>, rot=<0.0001 -0.0001 -0.7063 -0.7079>}
 foot_l_base_deform {parent=foot_l, loc=<1.0649 -8.2893 0.098>, rot=<0.4729 -0.5234 0.4807 -0.5208>}
 foot_l_dyn_deform {parent=foot_l_base_deform, loc=<0 0 0>, rot=<0 0 0 1>}
 calf_twist_02_l {parent=calf_l, loc=<-18.8064 -0.0967 0.0397>, rot=<0.0237 0.0095 -0.0035 -0.9997>}
 calf_twist_01_l {parent=calf_l, loc=<-35.813 -0.0967 0.0397>, rot=<0.0237 0.0095 -0.0035 -0.9997>}
 deform_knee_in_base_l {parent=calf_l, loc=<1.298 -5.2957 -0.0496>, rot=<0 0 0 -1>}
 deform_knee_in_l {parent=deform_knee_in_base_l, loc=<0 0 0>, rot=<0 0 0 1>}
 deform_knee_out_base_l {parent=calf_l, loc=<0 4 0>, rot=<0 0 0 -1>}
 deform_knee_out_l {parent=deform_knee_out_base_l, loc=<0 0 0>, rot=<0 0 0 1>}
 foot_l_null {parent=calf_l, type=Null, loc=<-42.561 0.1077 0.4292>, rot=<0.0418 -0.0455 0.0343 -0.9975>}
 thigh_twist_01_l {parent=thigh_l, loc=<-21.6303 0 0>, rot=<0.0013 0 0 -1>}
 deform_groin_base_l {parent=thigh_l, loc=<-2.1908 8.1142 -1.7442>, rot=<0 0 0 -1>}
 deform_groin_l {parent=deform_groin_base_l, loc=<0 0 0>, rot=<0 0 0 1>}
 SkirtA_l_01_Base {parent=thigh_l, loc=<-8.1282 8.1522 -4.8086>, rot=<-0.0525 -0.7873 -0.614 0.0193>}
 SkirtA_l_01_dyn {parent=SkirtA_l_01_Base, loc=<0 0 0>, rot=<0 0 0 1>}
 SkirtA_l_02_dyn {parent=SkirtA_l_01_dyn, loc=<9.9289 0 0>, rot=<0 0.0001 -0.0003 -1>}
 SkirtA_l_03_dyn {parent=SkirtA_l_02_dyn, loc=<10.1211 0 0>, rot=<0 0 0 1>}
 SkirtB_l_01_Base {parent=thigh_l, loc=<-9.0959 -3.371 -11.4006>, rot=<-0.0146 -0.9989 0.0272 0.0347>}
 SkirtB_l_01_dyn {parent=SkirtB_l_01_Base, loc=<0 0 0>, rot=<0 0 0 1>}
 SkirtB_l_02_dyn {parent=SkirtB_l_01_dyn, loc=<10.0991 0 0>, rot=<0 0.0007 -0.0022 -1>}
 SkirtB_l_03_dyn {parent=SkirtB_l_02_dyn, loc=<10.1602 0 0>, rot=<0 0 0 1>}
 SkirtC_l_01_Base {parent=thigh_l, loc=<-8.481 -15.2448 -4.7254>, rot=<0.0191 -0.8867 0.4604 0.037>}
 SkirtC_l_01_dyn {parent=SkirtC_l_01_Base, loc=<0 0 0>, rot=<0 0 0 1>}
 SkirtC_l_02_dyn {parent=SkirtC_l_01_dyn, loc=<9.9636 0 0>, rot=<0 -0.0027 -0.015 -0.9999>}
 SkirtC_l_03_dyn {parent=SkirtC_l_02_dyn, loc=<10.2454 0 0>, rot=<0 0 0 1>}
 calf_l_null {parent=thigh_l, type=Null, loc=<-39.0484 0 0>, rot=<0.0006 -0.0161 -0.0391 -0.9991>}
 thigh_r {parent=pelvis, loc=<-2.5693 -0.0109 7.7501>, rot=<0.0572 -0.0012 0.9982 0.0155>}
 calf_r {parent=thigh_r, loc=<39.0484 0 0>, rot=<0.0006 -0.0161 -0.0391 -0.9991>}
 foot_r {parent=calf_r, loc=<42.561 -0.1077 -0.4292>, rot=<0.0418 -0.0455 0.0343 -0.9975>}
 ball_r {parent=foot_r, loc=<6.1599 -13.8264 -0.6806>, rot=<0.0001 -0.0001 -0.7063 -0.7079>}
 foot_r_base_deform {parent=foot_r, loc=<-1.0649 8.2893 -0.098>, rot=<0.4807 0.5208 -0.4729 -0.5234>}
 foot_r_dyn_deform {parent=foot_r_base_deform, loc=<0 0 0>, rot=<0 0 0 1>}
 calf_twist_02_r {parent=calf_r, loc=<18.8064 0.0967 -0.0397>, rot=<0.0237 0.0095 -0.0035 -0.9997>}
 calf_twist_01_r {parent=calf_r, loc=<35.813 0.0967 -0.0397>, rot=<0.0237 0.0095 -0.0035 -0.9997>}
 deform_knee_in_base_r {parent=calf_r, loc=<-1.298 5.2957 0.0496>, rot=<0 0 0 -1>}
 deform_knee_in_r {parent=deform_knee_in_base_r, loc=<0 0 0>, rot=<0 0 0 1>}
 deform_knee_out_base_r {parent=calf_r, loc=<0 -4 0>, rot=<0 0 0 1>}
 deform_knee_out_r {parent=deform_knee_out_base_r, loc=<0 0 0>, rot=<0 0 0 1>}
 foot_r_null {parent=calf_r, type=Null, loc=<42.561 -0.1077 -0.4292>, rot=<0.0418 -0.0455 0.0343 -0.9975>}
 thigh_twist_01_r {parent=thigh_r, loc=<21.6303 0 0>, rot=<0.0013 0 0 -1>}
 SkirtA_r_01_Base {parent=thigh_r, loc=<8.1282 -8.1522 4.8085>, rot=<-0.0525 -0.7873 -0.614 0.0193>}
 SkirtA_r_01_dyn {parent=SkirtA_r_01_Base, loc=<0 0 0>, rot=<0 0 0 1>}
 SkirtA_r_02_dyn {parent=SkirtA_r_01_dyn, loc=<-9.9288 0 0>, rot=<0 0.0001 -0.0003 -1>}
 SkirtA_r_03_dyn {parent=SkirtA_r_02_dyn, loc=<-10.1212 0 0>, rot=<0 0 0 1>}
 SkirtB_r_01_Base {parent=thigh_r, loc=<9.0959 3.371 11.4005>, rot=<-0.0146 -0.9989 0.0272 0.0347>}
 SkirtB_r_01_dyn {parent=SkirtB_r_01_Base, loc=<0 0 0>, rot=<0 0 0 1>}
 SkirtB_r_02_dyn {parent=SkirtB_r_01_dyn, loc=<-10.0991 0 -0.0001>, rot=<0 0.0007 -0.0022 -1>}
 SkirtB_r_03_dyn {parent=SkirtB_r_02_dyn, loc=<-10.1602 0 0>, rot=<0 0 0 1>}
 SkirtC_r_01_Base {parent=thigh_r, loc=<8.481 15.2448 4.7254>, rot=<0.0191 -0.8867 0.4604 0.037>}
 SkirtC_r_01_dyn {parent=SkirtC_r_01_Base, loc=<0 0 0>, rot=<0 0 0 1>}
 SkirtC_r_02_dyn {parent=SkirtC_r_01_dyn, loc=<-9.9636 0 0>, rot=<0 -0.0027 -0.015 -0.9999>}
 SkirtC_r_03_dyn {parent=SkirtC_r_02_dyn, loc=<-10.2454 -0.0001 0>, rot=<0 0 0 1>}
 calf_r_null {parent=thigh_r, type=Null, loc=<39.0484 0 0>, rot=<0.0006 -0.0161 -0.0391 -0.9991>}
 deform_groin_base_r {parent=thigh_r, loc=<2.1908 -8.1142 1.7442>, rot=<0 0 0 -1>}
 deform_groin_r {parent=deform_groin_base_r, loc=<0 0 0>, rot=<0 0 0 1>}
 deform_glute_base_r {parent=pelvis, loc=<-8.8062 -8.8222 8.4287>, rot=<0.0501 -0.0275 0.8926 -0.4472>}
 deform_glute_r {parent=deform_glute_base_r, loc=<0 0 0>, rot=<0 0 0 1>}
 SkirtF_01_Base {parent=pelvis, loc=<-9.6585 12.5796 -0.0001>, rot=<-0.0325 -0.7082 -0.7046 -0.0323>}
 SkirtF_01_dyn {parent=SkirtF_01_Base, loc=<0 0 0>, rot=<0 0 0 1>}
 SkirtF_02_dyn {parent=SkirtF_01_dyn, loc=<9.8552 0 0>, rot=<0 -0.007 -0.0002 -1>}
 SkirtF_03_dyn {parent=SkirtF_02_dyn, loc=<10.0073 0 0>, rot=<0 0 0 1>}
 SkirtB_01_Base {parent=pelvis, loc=<-11.1176 -19.2318 -0.0035>, rot=<-0.011 -0.6938 0.72 0.0099>}
 SkirtB_01_dyn {parent=SkirtB_01_Base, loc=<0 0 0>, rot=<0 0 0 1>}
 SkirtB_02_dyn {parent=SkirtB_01_dyn, loc=<9.6586 0 0>, rot=<0 0 0 1>}
 SkirtB_03_dyn {parent=SkirtB_02_dyn, loc=<10.1849 0 0>, rot=<0 0 0 1>}
 thigh_l_null {parent=pelvis, type=Null, loc=<-2.5693 -0.0109 -7.7501>, rot=<0.0012 0.0572 0.0155 -0.9982>}
 thigh_r_null {parent=pelvis, type=Null, loc=<-2.5693 -0.0109 7.7501>, rot=<0.0572 -0.0012 0.9982 0.0155>}
 thigh_r_null_2 {parent=pelvis, type=Null, loc=<-2.5693 -0.0109 7.7501>, rot=<0.0572 -0.0012 0.9982 0.0155>}
 deform_glute_base_l {parent=pelvis, loc=<-8.8064 -8.8222 -8.4287>, rot=<0.0275 0.0501 -0.4472 -0.8926>}
 deform_glute_l {parent=deform_glute_base_l, loc=<0 0 0>, rot=<0 0 0 1>}
 spine_05_null {parent=pelvis, type=Null, loc=<37.9129 -4.2232 0>, rot=<0 0 0.058 -0.9983>}
 thigh_l_ctrl {controlType=EULER_TRANSFORM, type=Control, loc=<0 0 0>, rot=<0 0 0 1>}
 calf_l_ctrl {controlType=EULER_TRANSFORM, parent=thigh_l_ctrl, type=Control, loc=<0 0 0>, rot=<0 0 0 1>}
 thigh_r_ctrl {controlType=EULER_TRANSFORM, type=Control, loc=<0 0 0>, rot=<0 0 0 1>}
 calf_r_ctrl {controlType=EULER_TRANSFORM, parent=thigh_r_ctrl, type=Control, loc=<0 0 0>, rot=<0 0 0 1>}
]
variables [
 DBNO Carried {type=bool, default=False}
 Facial_Testing {type=double, default=0}
 Is Jumping {type=double, default=0}
 Is not crouching {type=double, default=0}
 Is_BreakDance_Emote {type=double, default=0}
 Is_Crouching {type=double, default=0}
 Is_Griddles_Emote {type=double, default=0}
 Is_on_Air {type=double, default=0}
 Is_pickaxe_Swing {type=double, default=0}
 Parachute_Open {type=double, default=0}
 PelvisProx {type=double, default=0}
 SPR L Calf Bkd {type=double, default=0}
 SPR L Calf Fixes {type=double, default=0}
 SPR L Ext Calf Bkd {type=double, default=0}
 SPR L Foot {type=double, default=0}
 SPR L Forward_Parachute {type=double, default=0}
 SPR L Leg Bkd {type=double, default=0}
 SPR L Leg Fwd {type=double, default=0}
 SPR L Leg Otw {type=double, default=0}
 SPR L Thigh Fixes {type=double, default=0}
 SPR R Calf Bkd {type=double, default=0}
 SPR R Calf Fixes {type=double, default=0}
 SPR R Ext Calf Bkd {type=double, default=0}
 SPR R Foot {type=double, default=0}
 SPR R Leg Bkd {type=double, default=0}
 SPR R Leg Fwd {type=double, default=0}
 SPR R Leg Otw {type=double, default=0}
 SPR R Thigh Fixes {type=double, default=0}
 is DBNO {type=bool, default=False}
 is_Jumping_CR {type=double, default=0}
 is_Moving_Back {type=double, default=0}
 is_Moving_Forward {type=double, default=0}
 is_Moving_left {type=double, default=0}
 is_Moving_right {type=double, default=0}
]
