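# Physics asset for the left-leg skirt pieces of the DuckCoast medium male player body.
# Two four-bone capsule chains, SkirtA and SkirtB, hang from the left thigh, with capsules widening along the dynamic segments.
# Base joints swing freely, while segment joints allow only a few degrees of swing, with SkirtB looser than SkirtA.
# Pelvis, thigh and calf capsules for both legs supply body shapes, linked by free constraints.
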PhysicsAsset /BRCosmetics/Characters/Player/Male/Medium/Bodies/M_MED_DuckCoast/Meshes/M_MED_DuckCoast_Physics_leftLeg
bodies {
 SkirtA_l_01_Base [{shapeType=Capsule, length=8, radius=2.5, shapeName=SkirtA_l_01_Base_capsule, center=<2.5 0 0>, rotation=<rot 90 0 0>}]
 SkirtA_l_01_dyn [{shapeType=Capsule, length=8, radius=6, shapeName=SkirtA_l_01_dyn_capsule, center=<6 0 4.5>, rotation=<rot 0 0 90>}]
 SkirtA_l_02_dyn [{shapeType=Capsule, length=8, radius=7, shapeName=SkirtA_l_02_dyn_capsule, center=<6 0 3.5>, rotation=<rot 0 0 90>}]
 SkirtA_l_03_dyn [{shapeType=Capsule, length=8, radius=8, shapeName=SkirtA_l_03_dyn_capsule, center=<8 0 2.5>, rotation=<rot 0 0 90>}]
 SkirtB_l_01_Base [{shapeType=Capsule, length=8, radius=2.5, shapeName=SkirtB_l_01_Base_capsule, center=<2.5 0 0>, rotation=<rot 90 0 0>}]
 SkirtB_l_01_dyn [{shapeType=Capsule, length=8, radius=6, shapeName=SkirtB_l_01_dyn_capsule, center=<4 0 6>, rotation=<rot 0 0 90>}]
 SkirtB_l_02_dyn [{shapeType=Capsule, length=8, radius=7, shapeName=SkirtB_l_02_dyn_capsule, center=<6 0 7.2>, rotation=<rot 0 0 90>}]
 SkirtB_l_03_dyn [{shapeType=Capsule, length=8, radius=8, shapeName=SkirtB_l_03_dyn_capsule, center=<8 0 7>, rotation=<rot 0 0 90>}]
 calf_l [{shapeType=Capsule, length=25, radius=8, shapeName=calf_l_capsule, center=<-20 0 0>, rotation=<rot 90 0 0>}]
 calf_r [{shapeType=Capsule, length=25, radius=8, shapeName=calf_r_capsule, center=<20 0 0>, rotation=<rot 90 180 0>}]
 pelvis [{shapeType=Capsule, length=10, radius=15, center=<0 -4.2232 0>, rotation=<rot 0 90 0>}]
 thigh_l [{shapeType=Capsule, length=25, radius=9, shapeName=thigh_l_capsule, center=<-25 -1.2 -1>, rotation=<rot 88.391 0.0004 0.0006>}]
 thigh_r [{shapeType=Capsule, length=25, radius=9, shapeName=thigh_r_capsule, center=<25 1.1824 0.5>, rotation=<rot 86.6937 0.0003 -173>}]
}
constraints [
 {
  bone1Name=SkirtA_l_01_Base
  bone2Name=thigh_l
  swing1LimitDegrees=45
  swing1Motion=Free
  swing2LimitDegrees=45
  swing2Motion=Free
  twistLimitDegrees=45
  twistMotion=Free
 }
 {bone1Name=SkirtA_l_01_dyn, bone2Name=SkirtA_l_01_Base, swing1LimitDegrees=0, swing2LimitDegrees=2.5, twistLimitDegrees=0}
 {bone1Name=SkirtA_l_02_dyn, bone2Name=SkirtA_l_01_dyn, swing1LimitDegrees=0, swing2LimitDegrees=2.5, twistLimitDegrees=0}
 {bone1Name=SkirtA_l_03_dyn, bone2Name=SkirtA_l_02_dyn, swing1LimitDegrees=0, swing2LimitDegrees=2.5, twistLimitDegrees=0}
 {
  bone1Name=SkirtB_l_01_Base
  bone2Name=thigh_l
  swing1LimitDegrees=45
  swing1Motion=Free
  swing2LimitDegrees=45
  swing2Motion=Free
  twistLimitDegrees=45
  twistMotion=Free
 }
 {bone1Name=SkirtB_l_01_dyn, bone2Name=SkirtB_l_01_Base, swing1LimitDegrees=5, swing2LimitDegrees=3, twistLimitDegrees=0}
 {bone1Name=SkirtB_l_02_dyn, bone2Name=SkirtB_l_01_dyn, swing1LimitDegrees=5, swing2LimitDegrees=4, twistLimitDegrees=0}
 {bone1Name=SkirtB_l_03_dyn, bone2Name=SkirtB_l_02_dyn, swing1LimitDegrees=5, swing2LimitDegrees=5, twistLimitDegrees=0}
 {
  bone1Name=thigh_l
  bone2Name=pelvis
  swing1LimitDegrees=45
  swing1Motion=Free
  swing2LimitDegrees=45
  swing2Motion=Free
  twistLimitDegrees=45
  twistMotion=Free
 }
 {
  bone1Name=calf_l
  bone2Name=thigh_l
  swing1LimitDegrees=45
  swing1Motion=Free
  swing2LimitDegrees=45
  swing2Motion=Free
  twistLimitDegrees=45
  twistMotion=Free
 }
 {
  bone1Name=thigh_r
  bone2Name=pelvis
  swing1LimitDegrees=45
  swing1Motion=Free
  swing2LimitDegrees=45
  swing2Motion=Free
  twistLimitDegrees=45
  twistMotion=Free
 }
 {
  bone1Name=calf_r
  bone2Name=thigh_r
  swing1LimitDegrees=45
  swing1Motion=Free
  swing2LimitDegrees=45
  swing2Motion=Free
  twistLimitDegrees=45
  twistMotion=Free
 }
]
