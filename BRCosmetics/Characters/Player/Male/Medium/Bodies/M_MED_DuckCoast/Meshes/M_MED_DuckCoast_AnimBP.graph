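# Post-pose AnimBP for the DuckCoast player body cosmetic, animating its coat skirt, bow and upper-arm secondary bones.
# The parent mesh's pose feeds a Control Rig, per-leg RigidBody sims, a skirt tweak and a weighted bow constraint, then gets cached.
# AnimDynamics and additive bone offsets finish the pose.
# The EventGraph turns movement, skydive and emote-montage states into rig floats and computes state-based gravity and rigid-body settings.
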AnimBlueprint /BRCosmetics/Characters/Player/Male/Medium/Bodies/M_MED_DuckCoast/Meshes/M_MED_DuckCoast_AnimBP
target_skeleton=&/M_MED_DuckCoast_Skeleton
graphs [
 AnimGraph {nodes [
   n1 = Anim Dynamics - Bone: SkirtC_l_01_dyn(ComponentPose=Local To Component(LocalPose=Use cached pose 'Initial Pose'()), Alpha=1, ExternalForce="0,0,0") {
    angularDampingOverride=1
    angularSpringConstant=35000
    bAngularSpring=true
    bOverrideAngularDamping=true
    bUseGravityOverride=true
    gravityScale=0.35
    simulationSpace=RootRelative
    boundBone {boneName=SkirtC_l_01_dyn}
    componentAppliedLinearAccClamp=<10000 10000 10000>
    componentLinearAccScale=<0.3 0.3 0.4>
    componentLinearVelScale=<0.3 0.3 0.4>
    gravityOverride {z=-980}
    physicsBodyDefinitions [{
      boundBone {boneName=SkirtC_l_01_dyn}
      boxExtents=<45 45 55>
      constraintSetup {angularLimitsMax=<0 0 0>, angularLimitsMin=<0 -15 0>, angularTarget=<1 0 0>}
      localJointOffset=<3 0 0>
     }]
   }
   n2 = Anim Dynamics - Bone: SkirtC_l_02_dyn(ComponentPose=@n1, Alpha=1, ExternalForce="0,0,0") {
    angularDampingOverride=1
    angularSpringConstant=35000
    bAngularSpring=true
    bOverrideAngularDamping=true
    bUseGravityOverride=true
    gravityScale=0.35
    simulationSpace=RootRelative
    boundBone {boneName=SkirtC_l_02_dyn}
    componentAppliedLinearAccClamp=<10000 10000 10000>
    componentLinearAccScale=<0.3 0.3 0.4>
    componentLinearVelScale=<0.3 0.3 0.4>
    gravityOverride {z=-980}
    physicsBodyDefinitions [{
      boundBone {boneName=SkirtC_l_02_dyn}
      boxExtents=<45 45 55>
      constraintSetup {angularLimitsMax=<0 0 0>, angularLimitsMin=<0 -10 0>, angularTarget=<1 0 0>}
      localJointOffset=<3 0 0>
     }]
   }
   n3 = Anim Dynamics - Bone: SkirtC_l_03_dyn(ComponentPose=@n2, Alpha=1, ExternalForce="0,0,0") {
    angularDampingOverride=1
    angularSpringConstant=35000
    bAngularSpring=true
    bOverrideAngularDamping=true
    bUseGravityOverride=true
    gravityScale=0.35
    simulationSpace=RootRelative
    boundBone {boneName=SkirtC_l_03_dyn}
    componentAppliedLinearAccClamp=<10000 10000 10000>
    componentLinearAccScale=<0.3 0.3 0.4>
    componentLinearVelScale=<0.3 0.3 0.4>
    gravityOverride {z=-980}
    physicsBodyDefinitions [{
      boundBone {boneName=SkirtC_l_03_dyn}
      boxExtents=<45 45 55>
      constraintSetup {angularLimitsMax=<0 0 0>, angularLimitsMin=<0 -5 0>, angularTarget=<1 0 0>}
      localJointOffset=<3 0 0>
     }]
   }
   n4 = Anim Dynamics - Bone: SkirtB_01_dyn(ComponentPose=@n3, Alpha=1, ExternalForce="0,0,0") {
    angularDampingOverride=1
    angularSpringConstant=35000
    bAngularSpring=true
    bOverrideAngularDamping=true
    bUseGravityOverride=true
    gravityScale=0.35
    simulationSpace=RootRelative
    boundBone {boneName=SkirtB_01_dyn}
    componentAppliedLinearAccClamp=<10000 10000 10000>
    componentLinearAccScale=<0.3 0.3 0.4>
    componentLinearVelScale=<0.3 0.3 0.4>
    gravityOverride {z=-980}
    physicsBodyDefinitions [{
      boundBone {boneName=SkirtB_01_dyn}
      boxExtents=<45 45 55>
      constraintSetup {angularLimitsMax=<0 0 0>, angularLimitsMin=<0 -15 0>, angularTarget=<1 0 0>}
      localJointOffset=<3 0 0>
     }]
   }
   n5 = Anim Dynamics - Bone: SkirtB_02_dyn(ComponentPose=@n4, Alpha=1, ExternalForce="0,0,0") {
    angularDampingOverride=1
    angularSpringConstant=35000
    bAngularSpring=true
    bOverrideAngularDamping=true
    bUseGravityOverride=true
    gravityScale=0.35
    simulationSpace=RootRelative
    boundBone {boneName=SkirtB_02_dyn}
    componentAppliedLinearAccClamp=<10000 10000 10000>
    componentLinearAccScale=<0.3 0.3 0.4>
    componentLinearVelScale=<0.3 0.3 0.4>
    gravityOverride {z=-980}
    physicsBodyDefinitions [{
      boundBone {boneName=SkirtB_02_dyn}
      boxExtents=<45 45 55>
      constraintSetup {angularLimitsMax=<0 0 0>, angularLimitsMin=<0 -10 0>, angularTarget=<1 0 0>}
      localJointOffset=<3 0 0>
     }]
   }
   n6 = Anim Dynamics - Bone: SkirtB_03_dyn(ComponentPose=@n5, Alpha=1, ExternalForce="0,0,0") {
    angularDampingOverride=1
    angularSpringConstant=35000
    bAngularSpring=true
    bOverrideAngularDamping=true
    bUseGravityOverride=true
    gravityScale=0.35
    simulationSpace=RootRelative
    boundBone {boneName=SkirtB_03_dyn}
    componentAppliedLinearAccClamp=<10000 10000 10000>
    componentLinearAccScale=<0.3 0.3 0.4>
    componentLinearVelScale=<0.3 0.3 0.4>
    gravityOverride {z=-980}
    physicsBodyDefinitions [{
      boundBone {boneName=SkirtB_03_dyn}
      boxExtents=<45 45 55>
      constraintSetup {angularLimitsMax=<0 0 0>, angularLimitsMin=<0 -5 0>, angularTarget=<1 0 0>}
      localJointOffset=<3 0 0>
     }]
   }
   n7 = Anim Dynamics - Bone: SkirtC_r_01_dyn(ComponentPose=@n6, Alpha=1, ExternalForce="0,0,0") {
    angularDampingOverride=1
    angularSpringConstant=35000
    bAngularSpring=true
    bOverrideAngularDamping=true
    bUseGravityOverride=true
    gravityScale=0.35
    simulationSpace=RootRelative
    boundBone {boneName=SkirtC_r_01_dyn}
    componentAppliedLinearAccClamp=<10000 10000 10000>
    componentLinearAccScale=<0.3 0.3 0.4>
    componentLinearVelScale=<0.3 0.3 0.4>
    gravityOverride {z=-980}
    physicsBodyDefinitions [{
      boundBone {boneName=SkirtC_r_01_dyn}
      boxExtents=<45 45 55>
      constraintSetup {angularLimitsMax=<0 0 0>, angularLimitsMin=<0 -15 0>, angularTarget=<1 0 0>}
      localJointOffset=<3 0 0>
     }]
   }
   n8 = Anim Dynamics - Bone: SkirtC_r_02_dyn(ComponentPose=@n7, Alpha=1, ExternalForce="0,0,0") {
    angularDampingOverride=1
    angularSpringConstant=35000
    bAngularSpring=true
    bOverrideAngularDamping=true
    bUseGravityOverride=true
    gravityScale=0.35
    simulationSpace=RootRelative
    boundBone {boneName=SkirtC_r_02_dyn}
    componentAppliedLinearAccClamp=<10000 10000 10000>
    componentLinearAccScale=<0.3 0.3 0.4>
    componentLinearVelScale=<0.3 0.3 0.4>
    gravityOverride {z=-980}
    physicsBodyDefinitions [{
      boundBone {boneName=SkirtC_r_02_dyn}
      boxExtents=<45 45 55>
      constraintSetup {angularLimitsMax=<0 10 0>, angularLimitsMin=<0 0 0>, angularTarget=<1 0 0>}
      localJointOffset=<3 0 0>
     }]
   }
   n9 = Anim Dynamics - Bone: SkirtC_r_03_dyn(ComponentPose=@n8, Alpha=1, ExternalForce="0,0,0") {
    angularDampingOverride=1
    angularSpringConstant=35000
    bAngularSpring=true
    bOverrideAngularDamping=true
    bUseGravityOverride=true
    gravityScale=0.35
    simulationSpace=RootRelative
    boundBone {boneName=SkirtC_r_03_dyn}
    componentAppliedLinearAccClamp=<10000 10000 10000>
    componentLinearAccScale=<0.3 0.3 0.4>
    componentLinearVelScale=<0.3 0.3 0.4>
    gravityOverride {z=-980}
    physicsBodyDefinitions [{
      boundBone {boneName=SkirtC_r_03_dyn}
      boxExtents=<45 45 55>
      constraintSetup {angularLimitsMax=<0 5 0>, angularLimitsMin=<0 0 0>, angularTarget=<1 0 0>}
      localJointOffset=<3 0 0>
     }]
   }
   n10 = Anim Dynamics - Bone: bow_c(ComponentPose=@n9, Alpha=1, ExternalForce="0,0,0") {
    angularDampingOverride=1
    angularSpringConstant=20000
    bAngularSpring=true
    bOverrideAngularDamping=true
    bUseGravityOverride=true
    gravityScale=0.35
    simulationSpace=BoneRelative
    boundBone {boneName=bow_c}
    componentLinearAccScale=<0.3 0.3 0.3>
    componentLinearVelScale=<0.3 0.3 0.3>
    gravityOverride {z=-980}
    physicsBodyDefinitions [{
      boundBone {boneName=bow_c}
      boxExtents=<34 45 40>
      constraintSetup {angularLimitsMax=<0 0 0>, angularLimitsMin=<0 0 0>, angularTarget=<1 0 0>}
      localJointOffset=<0 0 -4>
     }]
    relativeSpaceBone {boneName=spine_05}
   }
   n11 = Anim Dynamics - Bone: bow_l_dyn(ComponentPose=@n10, Alpha=1, ExternalForce="0,0,0") {
    angularDampingOverride=1
    angularSpringConstant=20000
    bAngularSpring=true
    bOverrideAngularDamping=true
    bUseGravityOverride=true
    gravityScale=0.35
    simulationSpace=BoneRelative
    boundBone {boneName=bow_l_dyn}
    componentLinearAccScale=<0.3 0.3 0.3>
    componentLinearVelScale=<0.3 0.3 0.3>
    gravityOverride {z=-980}
    physicsBodyDefinitions [{
      boundBone {boneName=bow_l_dyn}
      boxExtents=<30 40 50>
      constraintSetup {angularLimitsMax=<0 8 0>, angularLimitsMin=<0 -15 0>, angularTarget=<1 0 0>}
      localJointOffset=<2 0 0>
     }]
    relativeSpaceBone {boneName=spine_05}
   }
   n12 = Anim Dynamics - Bone: bow_r_dyn(ComponentPose=@n11, Alpha=1, ExternalForce="0,0,0") {
    angularDampingOverride=1
    angularSpringConstant=20000
    bAngularSpring=true
    bOverrideAngularDamping=true
    bUseGravityOverride=true
    gravityScale=0.35
    simulationSpace=BoneRelative
    boundBone {boneName=bow_r_dyn}
    componentLinearAccScale=<0.3 0.3 0.3>
    componentLinearVelScale=<0.3 0.3 0.3>
    gravityOverride {z=-980}
    physicsBodyDefinitions [{
      boundBone {boneName=bow_r_dyn}
      boxExtents=<30 40 50>
      constraintSetup {angularLimitsMax=<0 8 0>, angularLimitsMin=<0 -15 0>, angularTarget=<1 0 0>}
      localJointOffset=<-2 0 0>
     }]
    relativeSpaceBone {boneName=spine_05}
   }
   n13 = "Transform (Modify) Bone - Bone: SkirtC_r_01_dyn"(ComponentPose=@n12, Alpha=1, Translation="0,-5,-0.471", Rotation="0,0,0", Scale="1,1,1") {
    rotationMode=BMM_Additive
    scaleMode=BMM_Additive
    translationMode=BMM_Additive
    alphaScaleBiasClamp {bInterpResult=true, interpSpeedDecreasing=3, interpSpeedIncreasing=3}
    boneToModify {boneName=SkirtC_r_01_dyn}
    translation {y=-5, z=-0.471}
   }
   n14 = "Transform (Modify) Bone - Bone: SkirtC_r_02_dyn"(ComponentPose=@n13, Alpha=1, Translation="0.000,-5.000,2.058", Rotation="0,0,-75", Scale="1,1,1") {
    rotationMode=BMM_Additive
    translationMode=BMM_Additive
    alphaScaleBiasClamp {bInterpResult=true, interpSpeedDecreasing=3, interpSpeedIncreasing=3}
    boneToModify {boneName=SkirtC_r_02_dyn}
    rotation {roll=-75}
    translation {y=-5, z=2.058}
   }
   n15 = "Transform (Modify) Bone - Bone: SkirtC_r_03_dyn"(ComponentPose=@n14, Alpha=1, Translation="0,-5,-3.7", Rotation="0,0,0", Scale="1,1,1") {
    rotationMode=BMM_Additive
    translationMode=BMM_Additive
    alphaScaleBiasClamp {bInterpResult=true, interpSpeedDecreasing=3, interpSpeedIncreasing=3}
    boneToModify {boneName=SkirtC_r_03_dyn}
    translation {y=-5, z=-3.7}
   }
   n16 = "Transform (Modify) Bone - Bone: SkirtB_l_02_dyn"(ComponentPose=@n15, Alpha=1, Translation="0,0,0", Rotation="21.195,0.000,0.000", Scale="1,1,1") {
    rotationMode=BMM_Additive
    translationMode=BMM_Additive
    alphaScaleBiasClamp {bInterpResult=true, interpSpeedDecreasing=3, interpSpeedIncreasing=3}
    boneToModify {boneName=SkirtB_l_02_dyn}
    rotation {pitch=21.195}
   }
   n17 = "Transform (Modify) Bone - Bone: SkirtC_l_02_dyn"(ComponentPose=@n16, Alpha=1, Translation="0.000,-0.000,0.468", Rotation="4.407,2.369,48.339", Scale="1,1,1") {
    rotationMode=BMM_Additive
    translationMode=BMM_Additive
    alphaScaleBiasClamp {bInterpResult=true, interpSpeedDecreasing=3, interpSpeedIncreasing=3}
    boneToModify {boneName=SkirtC_l_02_dyn}
    rotation=<rot 4.407 2.369 48.339>
    translation {z=0.468}
   }
   n18 = "Transform (Modify) Bone - Bone: SkirtB_01_Base"(ComponentPose=@n17, Alpha=1, Translation="-0.339,-1.814,0.610", Rotation="0,0,0", Scale="1,1,1") {
    rotationMode=BMM_Additive
    translationMode=BMM_Additive
    alphaScaleBiasClamp {bInterpResult=true, interpSpeedDecreasing=3, interpSpeedIncreasing=3}
    boneToModify {boneName=SkirtB_01_Base}
    translation=<-0.339 -1.814 0.61>
   }
   n19 = "Transform (Modify) Bone - Bone: SkirtC_l_01_Base"(ComponentPose=@n18, Alpha=1, Translation="7.451,4.330,21.870", Rotation="0,0,0", Scale="1,1,1") {
    rotationMode=BMM_Additive
    translationMode=BMM_Additive
    alphaScaleBiasClamp {bInterpResult=true, interpSpeedDecreasing=5, interpSpeedIncreasing=3}
    boneToModify {boneName=SkirtC_l_01_Base}
    translation=<7.451 4.33 21.87>
   }
   n20 = "Transform (Modify) Bone - Bone: SkirtB_01_Base"(ComponentPose=@n19, Alpha=1, Translation="2.309,-1.554,-3.7", Rotation="-20.000,-0.000,-20.000", Scale="1,1,1") {
    rotationMode=BMM_Additive
    translationMode=BMM_Additive
    alphaScaleBiasClamp {bInterpResult=true, interpSpeedDecreasing=5, interpSpeedIncreasing=1}
    boneToModify {boneName=SkirtB_01_Base}
    rotation {pitch=-20, roll=-20}
    translation=<2.309 -1.554 -3.7>
   }
   n21 = "Transform (Modify) Bone - Bone: SkirtC_r_01_Base"(ComponentPose=@n20, Alpha=1, Translation="-0.000,1.415,9.812", Rotation="0.000,0.000,-70.000", Scale="1,1,1") {
    rotationMode=BMM_Additive
    translationMode=BMM_Additive
    alphaScaleBiasClamp {bInterpResult=true, interpSpeedDecreasing=5, interpSpeedIncreasing=3}
    boneToModify {boneName=SkirtC_r_01_Base}
    rotation {roll=-70}
    translation {y=1.415, z=9.812}
   }
   n22 = "Transform (Modify) Bone - Bone: SkirtC_l_02_dyn"(ComponentPose=@n21, Alpha=1, Translation="2.294,-0.000,0.878", Rotation="0,0,0", Scale="1,1,1") {
    rotationMode=BMM_Additive
    translationMode=BMM_Additive
    alphaScaleBiasClamp {bInterpResult=true, interpSpeedDecreasing=5, interpSpeedIncreasing=3}
    boneToModify {boneName=SkirtC_l_02_dyn}
    translation {x=2.294, z=0.878}
   }
   n23 = "Transform (Modify) Bone - Bone: SkirtB_02_dyn"(ComponentPose=@n22, Alpha=1, Translation="0.000,11.810,-8.649", Rotation="25.000,-0.000,75.000", Scale="1,1,1") {
    rotationMode=BMM_Additive
    translationMode=BMM_Additive
    alphaScaleBiasClamp {bInterpResult=true, interpSpeedDecreasing=5, interpSpeedIncreasing=2}
    boneToModify {boneName=SkirtB_02_dyn}
    rotation {pitch=25, roll=75}
    translation {y=11.81, z=-8.649}
   }
   n24 = "Transform (Modify) Bone - Bone: SkirtC_r_02_dyn"(ComponentPose=@n23, Alpha=1, Translation="-0.001,-4.137,-0.757", Rotation="0,0,0", Scale="1,1,1") {
    rotationMode=BMM_Additive
    translationMode=BMM_Additive
    alphaScaleBiasClamp {bInterpResult=true, interpSpeedDecreasing=5, interpSpeedIncreasing=3}
    boneToModify {boneName=SkirtC_r_02_dyn}
    translation=<-0.001 -4.137 -0.757>
   }
   n25 = "Transform (Modify) Bone - Bone: SkirtC_r_03_dyn"(ComponentPose=@n24, Alpha=1, Translation="-4.015,4.289,1.071", Rotation="20.075,-3.521,-5.249", Scale="1,1,1") {
    rotationMode=BMM_Additive
    translationMode=BMM_Additive
    alphaScaleBiasClamp {bInterpResult=true, interpSpeedDecreasing=5, interpSpeedIncreasing=3}
    boneToModify {boneName=SkirtC_r_03_dyn}
    rotation=<rot 20.075 -3.521 -5.249>
    translation=<-4.015 4.289 1.071>
   }
   n26 = "Transform (Modify) Bone - Bone: SkirtC_r_02_dyn"(ComponentPose=@n25, Alpha=1, Translation="0,0,0", Rotation="-35,-80,0.001", Scale="1,1,1") {
    rotationMode=BMM_Additive
    rotationSpace=BCS_BoneSpace
    translationMode=BMM_Additive
    translationSpace=BCS_BoneSpace
    alphaScaleBiasClamp {bInterpResult=true, interpSpeedDecreasing=3, interpSpeedIncreasing=3}
    boneToModify {boneName=SkirtC_r_02_dyn}
    rotation=<rot -35 -80 0.001>
   }
   n27 = "Transform (Modify) Bone - Bone: SkirtB_l_02_dyn"(ComponentPose=@n26, Alpha=1, Translation="0,0,0", Rotation="0,5,15", Scale="1,1,1") {rotationMode=BMM_Additive, translationMode=BMM_Additive, boneToModify {boneName=SkirtB_l_02_dyn}, rotation {roll=15, yaw=5}}
   n28 = Anim Dynamics - Bone: upperarm_l_low_dyn(ComponentPose=@n27, Alpha=1, ExternalForce="0,0,0") {
    angularDampingOverride=1
    angularSpringConstant=14000
    bAngularSpring=true
    bOverrideAngularDamping=true
    bUseGravityOverride=true
    gravityScale=0.35
    simulationSpace=BoneRelative
    boundBone {boneName=upperarm_l_low_dyn}
    componentAppliedLinearAccClamp=<10000 10000 10000>
    componentLinearAccScale=<1 1 1>
    componentLinearVelScale=<1 1 1>
    gravityOverride {z=-980}
    physicsBodyDefinitions [{
      boundBone {boneName=upperarm_l_low_dyn}
      boxExtents=<40 50 40>
      constraintSetup {angularLimitsMax=<5 -5 5>, angularLimitsMin=<-5 -25 -15>, angularTarget=<1 0 0>}
      localJointOffset=<3 0 0>
     }]
    relativeSpaceBone {boneName=upperarm_l}
   }
   n29 = Anim Dynamics - Bone: upperarm_r_low_dyn(ComponentPose=@n28, Alpha=1, ExternalForce="0,0,0") {
    angularDampingOverride=1
    angularSpringConstant=14000
    bAngularSpring=true
    bOverrideAngularDamping=true
    bUseGravityOverride=true
    gravityScale=0.35
    simulationSpace=BoneRelative
    boundBone {boneName=upperarm_r_low_dyn}
    componentAppliedLinearAccClamp=<10000 10000 10000>
    componentLinearAccScale=<1 1 1>
    componentLinearVelScale=<1 1 1>
    gravityOverride {z=-980}
    physicsBodyDefinitions [{
      boundBone {boneName=upperarm_r_low_dyn}
      boxExtents=<40 50 40>
      constraintSetup {angularLimitsMax=<5 -5 15>, angularLimitsMin=<-5 -15 -5>, angularTarget=<1 0 0>}
      localJointOffset=<3 0 0>
     }]
    relativeSpaceBone {boneName=upperarm_r}
   }
   Output Pose(Result=Component To Local(ComponentPose=@n29)) {layerGroup=None}
   n30 = Control Rig!(
    Source=Copy Pose From Mesh() {bUseAttachedParent=true}
    Alpha
    Is_Crouching:double=Get Crouching() -> (Crouching:double)
    is_Moving_left:double=Get strafeLeft float() -> (strafeLeft float:double)
    is_Moving_right:double=Get strafeRight float() -> (strafeRight float:double)
    is_Moving_Back:double=Get moveBackward() -> (moveBackward:double)
    is_Moving_Forward:double=Get moveForward() -> (moveForward:double)
    Is_pickaxe_Swing:double=Get Pickaxe_on() -> (Pickaxe_on:double)
    Parachute_Open:double=Get Is_Parachute_Open() -> (Is_Parachute_Open:double)
    Is Jumping:double=Get Is Jumping() -> (Is Jumping:double)
    Is_on_Air:double=Get Is_on_Air() -> (Is_on_Air:double)
    Is not crouching:double=Get Is Sprinting Float() -> (Is Sprinting Float:double)
    is DBNO:bool=Get bIsDBNO()
    Is_Griddles_Emote:double=Get IsGriddles_Emote() -> (IsGriddles_Emote:double)
    Is_BreakDance_Emote:double=Get is_Breakdance_Emote() -> (is_Breakdance_Emote:double)
    is_Jumping_CR:double=Get Is Jumping()
    Facial_Testing:double=Get Is_Facial_Testing() -> (Is_Facial_Testing:double)
    DBNO Carried:bool=Get bIsDBNOCarried()
   ) -> (Pose) {controlRigAssetReference {blueprintRigClass=&/M_MED_DuckCoast_CtrlRig.M_MED_DuckCoast_CtrlRig_C}}
   n31 = RigidBody!(
    ComponentPose=Local To Component(LocalPose=@n30)
    Alpha
    ExternalForce
    SimSpaceSettings:Sim Space Settings Structure=(WorldAlpha=0,VelocityScaleZ=1,DampingAlpha=1,MaxLinearVelocity=10000,MaxAngularVelocity=10000,MaxLinearAcceleration=10000,MaxAngularAcceleration=10000,ExternalLinearDragV=(X=0,Y=0,Z=0),ExternalLinearVelocity=(X=0,Y=0,Z=0),ExternalAngularVelocity=(X=0,Y=0,Z=0))
   ) -> (Pose) {
    bOverrideWorldGravity=true
    overridePhysicsAsset=&/M_MED_DuckCoast_Physics_leftLeg
    simulationSpace=BaseBoneSpace
    baseBoneRef {boneName=thigh_l}
    componentAppliedLinearAccClamp=<0 0 0>
    overrideWorldGravity {z=-980}
   }
   n32 = RigidBody(
    ComponentPose=@n31
    SimSpaceSettings=(WorldAlpha=0,VelocityScaleZ=1,DampingAlpha=1,MaxLinearVelocity=10000,MaxAngularVelocity=10000,MaxLinearAcceleration=10000,MaxAngularAcceleration=10000,ExternalLinearDragV=(X=0,Y=0,Z=0),ExternalLinearVelocity=(X=0,Y=0,Z=0),ExternalAngularVelocity=(X=0,Y=0,Z=0))
   ) {
    bOverrideWorldGravity=true
    overridePhysicsAsset=&/M_MED_DuckCoast_Physics_rightLeg
    simulationSpace=BaseBoneSpace
    baseBoneRef {boneName=thigh_r}
    componentAppliedLinearAccClamp=<0 0 0>
    overrideWorldGravity {z=-980}
   }
   n33 = "Transform (Modify) Bone - Bone: SkirtF_01_dyn"!(ComponentPose=@n32, bAlphaBoolEnabled:bool=True, Translation="0,0,0", Rotation="20,0,50", Scale="1,1,1") -> (Pose) {
    alphaInputType=Bool
    rotationMode=BMM_Additive
    scaleMode=BMM_Additive
    translationMode=BMM_Additive
    boneToModify {boneName=SkirtF_01_dyn}
    rotation {pitch=20, roll=50}
   }
   n34 = "Constraint - bow_base (PPPPPP)"!(
    ComponentPose=@n33
    Alpha=1
    ConstraintWeights_0=0.4327
    ConstraintWeights_1:float=0.37
    ConstraintWeights_2:float=0.1083
    ConstraintWeights_3:float=0.0595
    ConstraintWeights_4:float=0.0253
    ConstraintWeights_5:float=0.0042
   ) -> (Pose) {
    boneToModify {boneName=bow_base}
    constraintSetup [
     {targetBone {boneName=neck_02}}
     {targetBone {boneName=spine_05}}
     {targetBone {boneName=neck_01}}
     {targetBone {boneName=head}}
     {targetBone {boneName=spine_04}}
     {targetBone {boneName=spine_03}}
    ]
    constraintWeights [1,1,1,1,1,1]
   }
   Save cached pose 'Initial Pose'(Pose=Component To Local(ComponentPose=@n34)) {cacheName=Initial Pose}
  ]}
 EventGraph {nodes [
   Event BlueprintUpdateAnimation()
   > n1 = Sequence!(execute) -> (then_0, then_1, then_2:Exec, then_3:Exec)
   n2 = Sequence!(execute=@n1.then_3) -> (then_0, then_1, then_2, then_3, then_4:Exec, then_5:Exec, then_6:Exec, then_7:Exec, then_8:Exec)
   n3 = Branch(execute=@n2.then_3, Condition=OR Boolean(A=Get bIsSkydiveFloating(), B=Get bIsSkyDiving()))
   > Set Is_on_Air(Is_on_Air:double=1) -> (Output_Get:double)
   Set Is_on_Air(execute=@n3.else, Is_on_Air=0)
   n4 = GetAnimInstance(self=GetOwnerSkeletalMeshComponent())
   Is Valid(exec=@n2.then_4, InputObject=@n4)
   > n5 = Set Is_Griddles_Emote(Is_Griddles_Emote:bool=Montage_IsPlaying(self=@n4, Montage=/BRCosmetics/Animation/Game/MainPlayer/Emotes/Griddles/Emote_Griddle_CMM_M)) -> (Output_Get:bool)
   > n6 = Branch(Condition=@n5)
   > Set IsGriddles_Emote(IsGriddles_Emote:double=1) -> (Output_Get:double)
   Set IsGriddles_Emote(execute=@n6.else, IsGriddles_Emote=0)
   Set On_Ground(execute=@n2.then_1, On_Ground:bool=NOT Boolean(A=Get bIsSkyDiving())) -> (Output_Get:bool)
   Set Is_Parachute_Open(execute=@n2.then_2, Is_Parachute_Open:double="To Float (Boolean)"(InBool=Get bIsParachuteOpened())) -> (Output_Get:double)
   n7 = GetAnimInstance(self=GetOwnerSkeletalMeshComponent())
   Is Valid(exec=@n2.then_5, InputObject=@n7)
   > n8 = Set Is Breakdance_Emote(Is Breakdance_Emote:bool=Montage_IsPlaying(self=@n7, Montage=/BRCosmetics/Animation/Game/MainPlayer/Emotes/Emote_Breakdance_M1)) -> (Output_Get:bool)
   > n9 = Branch(Condition=@n8)
   > Set is_Breakdance_Emote(is_Breakdance_Emote:double=1) -> (Output_Get:double)
   Set is_Breakdance_Emote(execute=@n9.else, is_Breakdance_Emote=0)
   n10 = Branch(execute=@n2.then_7, Condition=AND Boolean(A=Get bIsTacticalSprinting(), B=Get bIsJumping()))
   > Delay(Duration=0.2)
   > Set isTacticalSprinting(isTacticalSprinting:double=1) -> (Output_Get:double)
   Set isTacticalSprinting(execute=@n10.else, isTacticalSprinting=0)
   n11 = Get bIsKneeSliding()
   n12 = Branch(execute=@n2.then_8, Condition=AND Boolean(A=Get bIsCrouching(), B=NOT Boolean(A=@n11)))
   > Set Crouching(Crouching:double=1) -> (Output_Get:double)
   Set Crouching(execute=@n12.else, Crouching=0)
   n13 = GetAnimInstance(self=GetOwnerSkeletalMeshComponent())
   Is Valid(exec=@n2.then_6, InputObject=@n13)
   > n14 = Set Facial_Testing(Facial_Testing:bool=Montage_IsPlaying(self=@n13, Montage=/Game/ArtTools/Tech_Art/TestTools/Emotes/Emote_FacialTesting_Montage)) -> (Output_Get:bool)
   > n15 = Branch(Condition=@n14)
   > Set Is_Facial_Testing(Is_Facial_Testing:double=1) -> (Output_Get:double)
   Set Is_Facial_Testing(execute=@n15.else, Is_Facial_Testing=0)
   n16 = OR Boolean(A=Get bIsSkydiveFloating(), B=Get bIsSkyDiving())
   n17 = Get bIsJumping()
   n18 = Get bIsFalling()
   n19 = Branch(execute=@n2.then_0, Condition=AND Boolean(A=OR Boolean(A=@n17, B=@n18), B=NOT Boolean(A=@n16)))
   > Set Is Jumping(Is Jumping:double=1) -> (Output_Get:double)
   Set Is Jumping(execute=@n19.else, Is Jumping=0)
   n20 = "Gravity Override State Function (Nativized)"!(
    execute=@n1.then_0
    CharacterPartAnimInstance:Custom Character Part Anim Instance Object Reference=Self-Reference()
    GravityOverrideParameterSet:Fort Gravity Override Parameter Set Structure=Get Gravity Override Params() -> (Gravity Override Params:Fort Gravity Override Parameter Set Structure)
    RigidBodyAnimNodeParameterSet:Fort Rigid Body Anim Node Parameter Set Structure=Get RBAN Params() -> (RBAN Params:Fort Rigid Body Anim Node Parameter Set Structure)
    DebugDraw:bool=false
    GameTimeInSeconds:float=GetGameTimeInSeconds()
    DebugJointName:Name=root
    StateLogicLayerName:Name=none
   ) -> (
    then
    OutGravityOverride:Vector
    OutCurrentState:Name=None
    OutGravityOverrideParameters:Fort Gravity Override Parameters Structure
    OutRigidBodyAnimNodeParameters:Fort Rigid Body Anim Node Parameters Structure
   )
   > Set Gravity Override(Gravity Override=@n20)
   > Set RBAN Settings Out(RBAN Settings Out=@n20.OutRigidBodyAnimNodeParameters)
   > n21 = Branch(Condition=Get bIsInFrontEnd())
   > Set SimlationSpace(SimlationSpace:AnimPhysSimSpaceType Enum=World) -> (Output_Get:AnimPhysSimSpaceType Enum)
   > Set AlphaFE(AlphaFE:double=0.3) -> (Output_Get:double)
   Set SimlationSpace(execute=@n21.else)
   > Set AlphaFE(AlphaFE=1)
   n22 = PlayerMovement(execute=@n1.then_2, AnimInstance=Self-Reference(), XDotProductCoefficient=1, YDotProductCoefficient=1, ZDotProductCoefficient=1)
   n23 = BreakVector(InVec=MakeVector(X=@n22, Y=@n22.YMovement))
   > Set strafeLeft float(strafeLeft float:double="Absolute (Float)"(A="Clamp (Float)"(Value=@n23, Min=-1, Max=0))) -> (Output_Get:double)
   > Set strafeRight float(strafeRight float:double="Clamp (Float)"(Value=@n23, Min=0, Max=1)) -> (Output_Get:double)
   > Set moveForward(moveForward:double="Clamp (Float)"(Value=@n23.Y, Min=0, Max=1)) -> (Output_Get:double)
   > Set moveBackward(moveBackward:double="Absolute (Float)"(A="Clamp (Float)"(Value=@n23.Y, Min=-1, Max=0))) -> (Output_Get:double)
   n24 = Get bIsCrouching()
   n25 = Branch(execute=@n1.then_1, Condition=AND Boolean(A=Get bIsMoving2D(), B=NOT Boolean(A=@n24)))
   > Set Is Sprinting Float(Is Sprinting Float:double=1) -> (Output_Get:double)
   Set Is Sprinting Float(execute=@n25.else, Is Sprinting Float=0)
   Event OnBeginCrouching()
   > Set Crouching(Crouching=1)
   Event OnEndCrouching()
   > Set Crouching(Crouching=0)
   Event OnBeginKneeSliding() {is=title K2Node_Event}
   > Set Is knee sliding(Is knee sliding:double=1) -> (Output_Get:double)
   Event OnBeginPickaxeSwing() {is=title K2Node_Event}
   > Set Pickaxe_on(Pickaxe_on:double=1) -> (Output_Get:double)
   Event OnEndKneeSliding() {is=title K2Node_Event}
   > Set Is knee sliding(Is knee sliding=0)
   Event OnEndPickaxeSwing() {is=title K2Node_Event}
   > Set Pickaxe_on(Pickaxe_on=0)
  ]}
]
variables [
 AlphaFE {type=number}
 Crouching {type=number}
 Facial_Testing {type=boolean}
 Gravity Override {type=Vector, default=<0 0 -980>}
 Gravity Override Params {
  type=FortGravityOverrideParameterSet
  default=(Skydive_GoingUp=(GravityOverride=()),Skydive_FloatingDown=(GravityOverride=()),Skydive_DivingDown=(GravityOverride=()),Skydive_Parachuting=(GravityOverride=()),OnGround_Standing=(GravityOverride=()),OnGround_Moving=(GravityOverride=()),OnGround_Moving_HighRotationRate=(GravityOverride=()),OnGround_Crouch_Moving=(GravityOverride=()),Emote_or_Melee=(GravityOverride=()),Swimming=(GravityOverride=()),Front_End=(GravityOverride=()),DBNO=(GravityOverride=()),DBNO_Carried=(GravityOverride=()),OnGround_Falling=(GravityOverride=()),Ziplining=(GravityOverride=()),Skydive_Surfing=(GravityOverride=()),Riding_Standing=(GravityOverride=()),Riding_Moving=(GravityOverride=()),Riding_Moving_Targeting=(GravityOverride=()),Tactical_Sprint=(GravityOverride=()),Rail_Grinding=(GravityOverride=()),)
 }
 GravityOverride {type=Vector}
 Is Breakdance_Emote {type=boolean}
 Is Jumping {type=number}
 Is knee sliding {type=number}
 Is Sprinting Float {type=number}
 is_Breakdance_Emote {type=number}
 Is_Facial_Testing {type=number}
 Is_Griddles_Emote {type=boolean}
 Is_on_Air {type=number}
 Is_Parachute_Open {type=number}
 IsGriddles_Emote {type=number}
 isTacticalSprinting {type=number}
 moveBackward {type=number}
 moveForward {type=number}
 On_Ground {type=boolean}
 Pickaxe_on {type=number}
 RBAN Params {
  type=FortRigidBodyAnimNodeParameterSet
  default=(Skydive_GoingUp=(RigidBodyAnimNode=(SimSpaceSettings=(WorldAlpha=1.000000,VelocityScaleZ=0.000000,MaxLinearVelocity=700.000000,MaxAngularVelocity=3.000000,MaxLinearAcceleration=300.000000,MaxAngularAcceleration=1000.000000,ExternalLinearDragV=(X=0.000000,Y=0.500000,Z=0.000000)))),Skydive_FloatingDown=(RigidBodyAnimNode=(ComponentLinearAccScale=(X=0.000000,Y=0.500000,Z=0.000000),ComponentLinearVelScale=(X=0.000000,Y=0.500000,Z=0.000000),SimSpaceSettings=(WorldAlpha=1.000000,VelocityScaleZ=0.000000,MaxLinearVelocity=700.000000,MaxAngularVelocity=3.000000,MaxLinearAcceleration=300.000000,MaxAngularAcceleration=1000.000000,ExternalLinearDragV=(X=0.000000,Y=0.500000,Z=0.000000)))),Skydive_DivingDown=(RigidBodyAnimNode=(ComponentLinearAccScale=(X=0.100000,Y=0.000000,Z=0.000000),ComponentLinearVelScale=(X=0.100000,Y=0.000000,Z=0.000000),SimSpaceSettings=(WorldAlpha=1.000000,VelocityScaleZ=0.000000,MaxLinearVelocity=700.000000,MaxAngularVelocity=3.000000,MaxLinearAcceleration=300.000000,MaxAngularAcceleration=1000.000000,ExternalLinearDragV=(X=0.000000,Y=0.500000,Z=0.000000)))),Skydive_Parachuting=(RigidBodyAnimNode=(ComponentLinearAccScale=(X=0.000000,Y=0.100000,Z=0.300000),ComponentLinearVelScale=(X=0.000000,Y=0.100000,Z=0.300000),SimSpaceSettings=(WorldAlpha=1.000000,VelocityScaleZ=0.000000,MaxLinearVelocity=700.000000,MaxAngularVelocity=3.000000,MaxLinearAcceleration=300.000000,MaxAngularAcceleration=1000.000000,ExternalLinearDragV=(X=0.000000,Y=0.500000,Z=0.000000)))),OnGround_Standing=(RigidBodyAnimNode=(ComponentLinearAccScale=(X=0.300000,Y=0.000000,Z=0.000000),ComponentLinearVelScale=(X=0.300000,Y=0.000000,Z=0.000000),SimSpaceSettings=(WorldAlpha=1.000000,VelocityScaleZ=0.000000,MaxLinearVelocity=700.000000,MaxAngularVelocity=3.000000,MaxLinearAcceleration=300.000000,MaxAngularAcceleration=1000.000000,ExternalLinearDragV=(X=0.000000,Y=0.500000,Z=0.000000)))),OnGround_Moving=(RigidBodyAnimNode=(ComponentLinearAccScale=(X=0.300000,Y=0.000000,Z=0.000000),ComponentLinearVelScale=(X=0.300000,Y=0.000000,Z=0.000000),SimSpaceSettings=(WorldAlpha=0.400000,VelocityScaleZ=0.050000,DampingAlpha=0.000000,MaxLinearVelocity=600.000000,MaxAngularVelocity=7.000000,MaxLinearAcceleration=300.000000,MaxAngularAcceleration=1200.000000,ExternalLinearDragV=(X=0.000000,Y=0.500000,Z=0.000000)))),OnGround_Moving_HighRotationRate=(RigidBodyAnimNode=(ComponentLinearAccScale=(X=0.300000,Y=0.000000,Z=0.000000),ComponentLinearVelScale=(X=0.300000,Y=0.000000,Z=0.000000),SimSpaceSettings=(WorldAlpha=1.000000,VelocityScaleZ=0.000000,MaxLinearVelocity=700.000000,MaxAngularVelocity=3.000000,MaxLinearAcceleration=300.000000,MaxAngularAcceleration=1000.000000,ExternalLinearDragV=(X=0.000000,Y=0.500000,Z=0.000000)))),OnGround_Crouch_Moving=(RigidBodyAnimNode=(SimSpaceSettings=(WorldAlpha=1.000000,VelocityScaleZ=0.000000,MaxLinearVelocity=700.000000,MaxAngularVelocity=3.000000,MaxLinearAcceleration=300.000000,MaxAngularAcceleration=1000.000000,ExternalLinearDragV=(X=0.000000,Y=0.500000,Z=0.000000)))),Emote_or_Melee=(RigidBodyAnimNode=(SimSpaceSettings=(WorldAlpha=1.000000,VelocityScaleZ=0.000000,MaxLinearVelocity=700.000000,MaxAngularVelocity=3.000000,MaxLinearAcceleration=300.000000,MaxAngularAcceleration=1000.000000,ExternalLinearDragV=(X=0.000000,Y=0.500000,Z=0.000000)))),Swimming=(RigidBodyAnimNode=(SimSpaceSettings=(WorldAlpha=1.000000,VelocityScaleZ=0.000000,MaxLinearVelocity=700.000000,MaxAngularVelocity=3.000000,MaxLinearAcceleration=300.000000,MaxAngularAcceleration=1000.000000,ExternalLinearDragV=(X=0.000000,Y=0.500000,Z=0.000000)))),Front_End=(RigidBodyAnimNode=(SimSpaceSettings=(WorldAlpha=0.400000,VelocityScaleZ=0.000000,MaxLinearVelocity=700.000000,MaxAngularVelocity=3.000000,MaxLinearAcceleration=300.000000,MaxAngularAcceleration=1000.000000,ExternalLinearDragV=(X=0.000000,Y=0.500000,Z=0.000000)))),DBNO=(RigidBodyAnimNode=(SimSpaceSettings=(WorldAlpha=1.000000,VelocityScaleZ=0.000000,MaxLinearVelocity=700.000000,MaxAngularVelocity=3.000000,MaxLinearAcceleration=300.000000,MaxAngularAcceleration=1000.000000,ExternalLinearDragV=(X=0.000000,Y=0.500000,Z=0.000000)))),DBNO_Carried=(RigidBodyAnimNode=(SimSpaceSettings=(WorldAlpha=1.000000,VelocityScaleZ=0.000000,MaxLinearVelocity=700.000000,MaxAngularVelocity=3.000000,MaxLinearAcceleration=300.000000,MaxAngularAcceleration=1000.000000,ExternalLinearDragV=(X=0.000000,Y=0.500000,Z=0.000000)))),OnGround_Falling=(RigidBodyAnimNode=(SimSpaceSettings=(WorldAlpha=1.000000,VelocityScaleZ=0.000000,MaxLinearVelocity=700.000000,MaxAngularVelocity=3.000000,MaxLinearAcceleration=300.000000,MaxAngularAcceleration=1000.000000,ExternalLinearDragV=(X=0.000000,Y=0.500000,Z=0.000000)))),Ziplining=(RigidBodyAnimNode=(SimSpaceSettings=(WorldAlpha=1.000000,VelocityScaleZ=0.000000,MaxLinearVelocity=700.000000,MaxAngularVelocity=3.000000,MaxLinearAcceleration=300.000000,MaxAngularAcceleration=1000.000000,ExternalLinearDragV=(X=0.000000,Y=0.500000,Z=0.000000)))),Skydive_Surfing=(RigidBodyAnimNode=(SimSpaceSettings=(WorldAlpha=1.000000,VelocityScaleZ=0.000000,MaxLinearVelocity=700.000000,MaxAngularVelocity=3.000000,MaxLinearAcceleration=300.000000,MaxAngularAcceleration=1000.000000,ExternalLinearDragV=(X=0.000000,Y=0.500000,Z=0.000000)))),Riding_Standing=(RigidBodyAnimNode=(SimSpaceSettings=(WorldAlpha=1.000000,VelocityScaleZ=0.000000,MaxLinearVelocity=700.000000,MaxAngularVelocity=3.000000,MaxLinearAcceleration=300.000000,MaxAngularAcceleration=1000.000000,ExternalLinearDragV=(X=0.000000,Y=0.500000,Z=0.000000)))),Riding_Moving=(RigidBodyAnimNode=(SimSpaceSettings=(WorldAlpha=1.000000,VelocityScaleZ=0.000000,MaxLinearVelocity=700.000000,MaxAngularVelocity=3.000000,MaxLinearAcceleration=300.000000,MaxAngularAcceleration=1000.000000,ExternalLinearDragV=(X=0.000000,Y=0.500000,Z=0.000000)))),Riding_Moving_Targeting=(RigidBodyAnimNode=(SimSpaceSettings=(WorldAlpha=1.000000,VelocityScaleZ=0.000000,MaxLinearVelocity=700.000000,MaxAngularVelocity=3.000000,MaxLinearAcceleration=300.000000,MaxAngularAcceleration=1000.000000,ExternalLinearDragV=(X=0.000000,Y=0.500000,Z=0.000000)))),Tactical_Sprint=(RigidBodyAnimNode=(ComponentLinearAccScale=(X=0.300000,Y=0.000000,Z=0.000000),ComponentLinearVelScale=(X=0.300000,Y=0.000000,Z=0.000000),SimSpaceSettings=(WorldAlpha=1.000000,VelocityScaleZ=0.000000,MaxLinearVelocity=700.000000,MaxAngularVelocity=3.000000,MaxLinearAcceleration=300.000000,MaxAngularAcceleration=1000.000000,ExternalLinearDragV=(X=0.000000,Y=0.500000,Z=0.000000)))),Rail_Grinding=(RigidBodyAnimNode=(SimSpaceSettings=(WorldAlpha=1.000000,VelocityScaleZ=0.000000,MaxLinearVelocity=700.000000,MaxAngularVelocity=3.000000,MaxLinearAcceleration=300.000000,MaxAngularAcceleration=1000.000000,ExternalLinearDragV=(X=0.000000,Y=0.500000,Z=0.000000)))),)
 }
 RBAN Settings Out {type=FortRigidBodyAnimNodeParameters}
 SimlationSpace {type=AnimPhysSimSpaceType}
 strafeLeft float {type=number}
 strafeRight float {type=number}
]
