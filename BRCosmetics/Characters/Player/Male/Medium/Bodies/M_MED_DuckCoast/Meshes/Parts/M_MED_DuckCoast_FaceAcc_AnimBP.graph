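# Animation Blueprint for the DuckCoast male medium character's face accessory mesh, which follows the pose of its attached parent.
# A weighted multi-target constraint positions base_FaceAcc from the head, dyn_FaceAcc, two neck bones and spine_05, weighted mostly to the head.
# A fixed additive translation then offsets dyn_FaceAcc.
# The update event sets Is_on_Air while skydiving or floating, though the AnimGraph does not read it.
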AnimBlueprint /BRCosmetics/Characters/Player/Male/Medium/Bodies/M_MED_DuckCoast/Meshes/Parts/M_MED_DuckCoast_FaceAcc_AnimBP
target_skeleton=&/M_MED_DuckCoast_FaceAcc_Skeleton
graphs [
 AnimGraph {nodes [
   n1 = "Constraint - base_FaceAcc (PPPPP)"!(
    ComponentPose=Local To Component(LocalPose=Copy Pose From Mesh() {bUseAttachedParent=true})
    Alpha=1
    ConstraintWeights_0=0.7201
    ConstraintWeights_1:float=0.1854
    ConstraintWeights_2:float=0.0692
    ConstraintWeights_3:float=0.0175
    ConstraintWeights_4:float=0.0077
   ) -> (Pose) {
    boneToModify {boneName=base_FaceAcc}
    constraintSetup [
     {targetBone {boneName=head}}
     {targetBone {boneName=dyn_FaceAcc}}
     {targetBone {boneName=neck_01}}
     {targetBone {boneName=spine_05}}
     {targetBone {boneName=neck_02}}
    ]
    constraintWeights [1,1,1,1,1]
   }
   Output Pose(Result=Component To Local(ComponentPose="Transform (Modify) Bone - Bone: dyn_FaceAcc"(ComponentPose=@n1, Alpha=1, Translation="0,6,10", Rotation="0,0,0", Scale="1,1,1") {translationMode=BMM_Additive, boneToModify {boneName=dyn_FaceAcc}, translation {y=6, z=10}})) {layerGroup=None}
  ]}
 EventGraph {nodes [
   Event BlueprintUpdateAnimation()
   > n1 = Branch(Condition=OR Boolean(A=Get bIsSkydiveFloating(), B=Get bIsSkyDiving()))
   > Set Is_on_Air(Is_on_Air:double=1) -> (Output_Get:double)
   Set Is_on_Air(execute=@n1.else, Is_on_Air=0)
  ]}
]
variables [Is_on_Air {type=number}]
